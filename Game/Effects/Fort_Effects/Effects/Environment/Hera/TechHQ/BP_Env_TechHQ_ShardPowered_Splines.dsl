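; TechHQ environment effect Blueprint for "shard powered" splines, holding a function that builds twisted cluster meshes along a spline.
; SetupSplineMeshes adds a spline mesh component between each pair of points, using straight meshes except for a frayed final segment.
; Segments take local point positions, start arrive and end leave tangents; only the first requests a mesh update, via start scale.
; BeginPlay and the construction script are empty, and nothing here calls the function.
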(event EventBeginPlay)

(fn SetupSplineMeshes
  (params (Spline :type Spline%20Component%20Object%20Reference))
  (bind _returnvalue (Spline|GetNumberOfSplinePoints Spline))
  (bind _returnvalue_2 (- _returnvalue 2))
  (for _index (range (+ _returnvalue_2 1))
    (bind _returnvalue_3 (AddComponent|NODE_AddSplineMeshComponent-0))
    (Components|StaticMesh|SetStaticMesh _returnvalue_3 (select (== _index _returnvalue_2) "/Game/Effects/Fort_Effects/Mesh/Splines/SM_Spline_Cluster_Twist_Frayed.SM_Spline_Cluster_Twist_Frayed" "/Game/Effects/Fort_Effects/Mesh/Splines/SM_Spline_Cluster_Twist_Straight.SM_Spline_Cluster_Twist_Straight"))
    (bind (_returnvalue_inputkey _returnvalue_position _returnvalue_arrivetangent) (Utilities|Struct|BreakSplinePoint (Spline|GetSplinePointAt Spline _index "Local")))
    (SplineMesh|SetStartPosition _returnvalue_3 _returnvalue_position false)
    (SplineMesh|SetStartTangent _returnvalue_3 _returnvalue_arrivetangent false :!class /Script/Engine.SplineMeshComponent)
    (bind (_returnvalue_inputkey_2 _returnvalue_position_2 _returnvalue_arrivetangent_2 _returnvalue_leavetangent) (Utilities|Struct|BreakSplinePoint (Spline|GetSplinePointAt Spline (+ _index 1) "Local")))
    (SplineMesh|SetEndPosition _returnvalue_3 _returnvalue_position_2 false)
    (SplineMesh|SetEndTangent _returnvalue_3 _returnvalue_leavetangent false :!class /Script/Engine.SplineMeshComponent)
    (if (== _index 0)
      (SplineMesh|SetStartScale _returnvalue_3 "(X=1.000000,Y=1.000000)" true :!class /Script/Engine.SplineMeshComponent))))

(fn ConstructionScript)
