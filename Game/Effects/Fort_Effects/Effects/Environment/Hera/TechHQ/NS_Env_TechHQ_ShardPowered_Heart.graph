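# Niagara environment effect for the shard-powered heart at the Hera TechHQ location, with emitters named for tendrils, light fixture glows and a big glow.
# System-level variable assignments and time-of-day color, scalar and curve inputs are used, alongside IsOpen and IsClosed user booleans, both defaulting true.
# Only Culling_BugFix lists a renderer, a sprite with basic initialize and state modules, apparently a culling workaround; the other three emitters show no modules or renderers.
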NiagaraSystem /Game/Effects/Fort_Effects/Effects/Environment/Hera/TechHQ/NS_Env_TechHQ_ShardPowered_Heart
dependencies {
 usedDataInterfaces [&NiagaraDataInterfaceCurve]
 usedDynamicInputs [
  &FNTimeOfDayColor
  &FNTimeOfDayScalar
  &MakeLinearColor
  &FloatFromCurve
 ]
 usedModules [
  &self:SystemSpawnScript.SystemScriptSource.SystemScriptGraph.NiagaraNodeAssignment_11.SetVariables_#1
  &SystemState
  &self:SystemSpawnScript.SystemScriptSource.SystemScriptGraph.NiagaraNodeAssignment_12.SetVariables_#2
  &self:SystemSpawnScript.SystemScriptSource.SystemScriptGraph.NiagaraNodeAssignment_10.SetVariables_#3
  &self:SystemSpawnScript.SystemScriptSource.SystemScriptGraph.NiagaraNodeAssignment_13.SetVariables_#4
 ]
}
emitters {
 BigGlow {rendererClasses [], renderers []}
 Culling_BugFix {
  emitterUpdateScript [&EmitterState()]
  particleSpawnScript [&InitializeParticle(Color Mode=::NewEnumerator0, Lifetime=5, UsePositionOffset=false, Write Material Random=true)]
  particleUpdateScript [&ParticleState()]
  rendererClasses [&NiagaraSpriteRendererProperties]
  renderers [{rendererClass=&NiagaraSpriteRendererProperties}]
 }
 Light_Fixture_Glows {rendererClasses [], renderers []}
 Tendrils {rendererClasses [], renderers []}
}
summary {
 systemName=NS_Env_TechHQ_ShardPowered_Heart
 emitters [
  {emitterName=Tendrils, rendererClasses []}
  {emitterName=Light_Fixture_Glows, rendererClasses []}
  {emitterName=BigGlow, rendererClasses []}
  {emitterName=Culling_BugFix, rendererClasses [&NiagaraSpriteRendererProperties]}
 ]
 userVariables [
  User.IsClosed {type=NiagaraBool, defaultValue=true}
  User.IsOpen {type=NiagaraBool, defaultValue=true}
 ]
}
