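# Niagara system for the detail effects of the TechHQ shard-powered heart: a glowing blue-purple energy core in the environment.
# A pulsing, rotating orb, plasma meshes and radial ray sprites form the core.
# Orbiting debris, spawn-group-masked energy meshes and outward-moving spline meshes add motion around it.
# Owner LOD distance, a system distance fade and time-of-day inputs modulate scale, alpha and color; the older Rays emitter is disabled.
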NiagaraSystem /Game/Effects/Fort_Effects/Effects/Environment/Hera/TechHQ/NS_Env_TechHQ_ShardPowered_Heart_Details
dependencies {
 usedDataInterfaces [
  &NiagaraDataInterfaceVectorCurve
  &NiagaraDataInterfaceCurve
  &NiagaraDataInterfaceArrayColor
  &NiagaraDataInterfaceColorCurve
  &NiagaraDataInterfaceCamera
 ]
 usedDynamicInputs [
  &Add_Float
  &ModuloFloat
  &ClampFloat
  &Divide_Float
  &Subtract_Float
  &VectorLength
  &Subtract_Position
  &ReturnCameraPosition
  &Lerp_Float
  &MakeCustomIntFromBool
  &MultiplyLinearColor
  &MakeVector4_FromLinearColor
  &MakeVector4FromVectorAndFloat
  &MakeLinearColor
  &Multiply_Vector2D_ByFloat
 ]
 usedModules [
  &SystemState
  &self:SystemSpawnScript.SystemScriptSource.SystemScriptGraph.NiagaraNodeAssignment_3.SetVariables_#1
  &self:SystemSpawnScript.SystemScriptSource.SystemScriptGraph.NiagaraNodeAssignment_4.SetVariables_#2
  &self:SystemSpawnScript.SystemScriptSource.SystemScriptGraph.NiagaraNodeAssignment_5.SetVariables_#3
  &self:SystemSpawnScript.SystemScriptSource.SystemScriptGraph.NiagaraNodeAssignment_6.SetVariables_#4
  &self:SystemSpawnScript.SystemScriptSource.SystemScriptGraph.NiagaraNodeAssignment_7.SetVariables_#5
  &self:SystemSpawnScript.SystemScriptSource.SystemScriptGraph.NiagaraNodeAssignment_8.SetVariables_#6
  &self:SystemSpawnScript.SystemScriptSource.SystemScriptGraph.NiagaraNodeAssignment_9.SetVariables_#7
  &self:SystemSpawnScript.SystemScriptSource.SystemScriptGraph.NiagaraNodeAssignment_10.SetVariables_#8
  &self:SystemSpawnScript.SystemScriptSource.SystemScriptGraph.NiagaraNodeAssignment_11.SetVariables_#9
  &self:SystemSpawnScript.SystemScriptSource.SystemScriptGraph.NiagaraNodeAssignment_12.SetVariables_#10
 ]
}
emitters {
 Central_Flare {rendererClasses [], renderers []}
 Debris {
  emitterUpdateScript [
   &EmitterState(Inactive Response=::NewEnumerator1, Life Cycle Mode=::NewEnumerator1, Loop Duration=4, UseLoopDelay=false)
   &SpawnRate(Spawn Probability=0.75, SpawnRate=6, Use Spawn Probability=true)
  ]
  particleSpawnScript [
   /Game/Effects/Niagara/Modules/Orbit/FN_Orbit_Init_Variables_v2()
   &InitializeParticle(
    Color=<rgba 0.9513 0.5722 1 1>
    Lifetime=&RandomRangeFloat(Minimum=5.5, Maximum=7)
    Mesh Scale Mode=::NewEnumerator1
    Mesh Uniform Scale Max=6
    Mesh Uniform Scale Min=3
    UsePositionOffset=false
    Write Material Random=true
   ) {module=InitializeParticle001}
   &AddVelocityInCone(
    Cone Angle=18
    Cone Axis=<0 0 1>
    Module.Use Velocity Falloff On Cone Axis=false
    Velocity Distribution Along Cone Axis=0.75
    Velocity Strength=&RandomRangeFloat(Minimum=0, Maximum=20)
   )
   /SetVariables(Particles.Visibilitytag:NiagaraInt32=&RandomRangeInt(Minimum=0, Maximum=2)) {isSetParameters=true}
   &InitialMeshOrientation(Use Rotation Vector=false) {module=InitialMeshOrientation001}
   /SetVariables(Particles.CrystalPosition:NiagaraPosition=$Particles.Position, Particles.InitialPosition:NiagaraPosition=$Particles.Position) {isSetParameters=true}
  ]
  particleUpdateScript [
   &ParticleState()
   &Drag(Drag=0.25, Use Linear Drag=true, Use Rotational Drag=false)
   &ScaleColor(
    Scale Alpha=&FloatFromCurve(FloatCurve=DataInterface() {propertyValues {Curve {keys [<at -0.0046 1.0113 RCIM_Cubic>,<at 1.0011 0.002 RCIM_Cubic>]}}})
    ScaleA=true
    ScaleRGB=true
   )
   /Game/Effects/Niagara/Modules/Orbit/FN_Orbit_v4(
    Height min/max:Vector2f=<-400 100>
    Radius min/max:Vector2f=<100 300>
    Rotation Offset min/max:Vector2f=<0 1>
    Rotation Rate min/max:Vector2f=<0.01 0.08>
   )
   &UpdateMeshOrientation(Rotation Rate=&RandomRangeFloat(Minimum=-0.1, Maximum=0.1), Rotation Vector=<1 1 1>)
   &SolveForcesAndVelocity(Clamp Velocity=false, Limit Acceleration=false)
   &ScaleMeshSize(Scale Factor=&VectorFromFloat(Value=&FloatFromCurve(
      FloatCurve=DataInterface() {propertyValues {Curve {keys [
          {arriveTangent=5.1601, interpMode=RCIM_Cubic, leaveTangent=5.1601, tangentMode=RCTM_User, time=0, value=0}
          {arriveTangent=-0.0035, interpMode=RCIM_Cubic, leaveTangent=-0.0035, tangentMode=RCTM_User, time=0.2606, value=1}
          {arriveTangent=0.0043, interpMode=RCIM_Cubic, leaveTangent=0.0043, tangentMode=RCTM_User, time=0.4342, value=0.9932}
          {arriveTangent=-4.8059, tangentMode=RCTM_User, time=0.75, value=0}
         ]}}}
      Scale Curve=&FloatFromCurve(
       FloatCurve=DataInterface() {propertyValues {Curve {keys [{arriveTangent=-0.0001, interpMode=RCIM_Cubic, time=10000, value=1}, {arriveTangent=-0.0003, tangentMode=RCTM_User, time=15000, value=0}]}}}
       CurveIndex=$Engine.Owner.LODDistance
      )
     )))
   /SetVariables(Particles.TargetPosition:Vector3f=&Add_Vector(
     A=Linked() {linkedVariable {name=Engine.Owner.Position, type=NiagaraPosition}}
     B=&Subtract_Vector(A=$Particles.FN_Orbit_v4.Position, B=$Particles.Initial.InitialPosition)
    )) {isSetParameters=true}
   &self:Lerp(LerpSpeed:NiagaraFloat=&RandomRangeFloat(Minimum=0.01, Maximum=0.5)) {module=Lerp}
   &JitterPosition(Jitter Amount:NiagaraFloat=&Sine(Normalized Angle=$Emitter.Age, Period=4, Scale=4, Bias=2)) {enabled=false}
  ]
  rendererClasses [&NiagaraMeshRendererProperties]
  renderers [
   {rendererClass=&NiagaraMeshRendererProperties, rendererIndex=0}
   {rendererClass=&NiagaraMeshRendererProperties, rendererIndex=1}
   {rendererClass=&NiagaraMeshRendererProperties, rendererIndex=2}
  ]
 }
 Energy {
  emitterUpdateScript [
   /SetVariables(Emitter.PanSpeed=0.5) {isSetParameters=true}
   &EmitterState()
   &SpawnRate(Use Spawn Probability=false)
  ]
  particleSpawnScript [
   &InitializeParticle(
    Color=&MaskLinearColorBySpawnGroup(
     Masked Linear Color=&RandomRangeLinearColor(Minimum=<rgba 2.5648 0.569 5 1>, Maximum=<rgba 5 1.0977 4.2338 1>)
     Passthrough Linear Color=&MaskLinearColorBySpawnGroup(
      Masked Linear Color=&RandomRangeLinearColor(Minimum=<rgba 1 0.0336 0.9623 1>, Maximum=<rgba 0.7292 0.745 1 1>)
      Passthrough Linear Color=&MaskLinearColorBySpawnGroup(
       Masked Linear Color=<rgba 1.2844 1.0153 4 0.75>
       Passthrough Linear Color=&MaskLinearColorBySpawnGroup(
        Masked Linear Color=&SelectLinearColorFromArray(
         Array Sampling Mode=::NewEnumerator1
         Direct Array Index=&RandomRangeInt(Minimum=0, Maximum=1)
         Color Selection Array=DataInterface() {propertyValues {ColorData [<rgba 0.3199 0.5916 1 1>, <rgba 0.9443 0.3385 1 1>]}}
         Random Seed=?
        )
        Passthrough Linear Color=&MaskLinearColorBySpawnGroup(
         Masked Linear Color=<rgba 0.1907 0 1 0.25>
         Passthrough Linear Color=&MaskLinearColorBySpawnGroup(Masked Linear Color=<rgba 1 1 1 1>, Passthrough Linear Color=<rgba 1 1 1 1>, Spawn Group Mask=5)
         Spawn Group Mask=4
        )
        Spawn Group Mask=3
       )
       Spawn Group Mask=2
      )
      Spawn Group Mask=1
     )
    )
    Lifetime=&MaskFloatBySpawnGroup(
     Masked Float=2
     Passthrough Float=&MaskFloatBySpawnGroup(
      Masked Float=&RandomRangeFloat(Minimum=0.5, Maximum=1)
      Passthrough Float=&MaskFloatBySpawnGroup(
       Masked Float=&RandomRangeFloat(Minimum=2, Maximum=3)
       Passthrough Float=&MaskFloatBySpawnGroup(
        Masked Float=&RandomRangeFloat(Minimum=3, Maximum=4)
        Passthrough Float=&MaskFloatBySpawnGroup(
         Masked Float=4
         Passthrough Float=&MaskFloatBySpawnGroup(Masked Float=2, Passthrough Float=&RandomRangeFloat(Minimum=1, Maximum=2), Spawn Group Mask=5)
         Spawn Group Mask=4
        )
        Spawn Group Mask=3
       )
       Spawn Group Mask=2
      )
      Spawn Group Mask=1
     )
    )
    Mesh Scale Mode=::NewEnumerator1
    Mesh Uniform Scale Max=1
    Mesh Uniform Scale Min=0.5
    Sprite Size Max=<125 125>
    Sprite Size Min=<85 85>
    Sprite Size Mode=::NewEnumerator2
    UsePositionOffset=false
    Write Material Random=true
   )
   &InitialMeshOrientation(
    Mesh Orientation Mode=::NewEnumerator2
    Orientation Vector=&MaskVectorBySpawnGroup(
     Masked Vector=<0 0 1>
     Passthrough Vector=&MaskVectorBySpawnGroup(
      Masked Vector=&RandomVector(Vector Scale=1)
      Passthrough Vector=&MaskVectorBySpawnGroup(
       Masked Vector=&RandomVector(Vector Scale=1)
       Passthrough Vector=&MaskVectorBySpawnGroup(Masked Vector=<1 0 0>, Passthrough Vector=<0 0 0>, Spawn Group Mask=3)
       Spawn Group Mask=2
      )
      Spawn Group Mask=1
     )
    )
    Rotation=&MaskVectorBySpawnGroup(
     Masked Vector=<0 0 0>
     Passthrough Vector=&MaskVectorBySpawnGroup(
      Masked Vector=<0 0 0>
      Passthrough Vector=&MaskVectorBySpawnGroup(
       Masked Vector=<0 0 0>
       Passthrough Vector=&MaskVectorBySpawnGroup(
        Masked Vector=&RandomRangeVector(Minimum=<-1 -1 -1>, Maximum=<1 1 1>)
        Passthrough Vector=&RandomRangeVector(Minimum=<-0.7 0.25 0>, Maximum=<0.7 0.5 0>)
        Spawn Group Mask=3
       )
       Spawn Group Mask=2
      )
      Spawn Group Mask=1
     )
    )
    Use Rotation Vector=true
   )
   &ShapeLocation(
    MaskSpawnGroup=false
    Offset=<0 0 -200>
    Offset Mode=::NewEnumerator0
    Override Seed=false
    Sphere Radius=&MaskFloatBySpawnGroup(
     Masked Float=0
     Passthrough Float=&MaskFloatBySpawnGroup(
      Masked Float=400
      Passthrough Float=&MaskFloatBySpawnGroup(
       Masked Float=0
       Passthrough Float=&MaskFloatBySpawnGroup(
        Masked Float=0
        Passthrough Float=&MaskFloatBySpawnGroup(
         Masked Float=0
         Passthrough Float=&MaskFloatBySpawnGroup(Masked Float=0, Passthrough Float=400, Spawn Group Mask=5)
         Spawn Group Mask=4
        )
        Spawn Group Mask=3
       )
       Spawn Group Mask=2
      )
      Spawn Group Mask=1
     )
    )
    UseOwnerScale=true
   )
  ]
  particleUpdateScript [
   &ParticleState()
   &DynamicMaterialParameters(
    Index 0 Param 1=&FloatFromCurve(
     FloatCurve=DataInterface() {propertyValues {Curve {keys [<at 0 0.3 RCIM_Cubic>, {interpMode=RCIM_Cubic, tangentMode=RCTM_Break, time=0.9, value=1}]}}}
     Scale Curve=3
    )
    Index 0 Param 2=&FloatFromCurve(
     FloatCurve=DataInterface() {propertyValues {Curve {keys [
         {arriveTangent=-0.5351, interpMode=RCIM_Cubic, leaveTangent=-0.5351, tangentMode=RCTM_User, time=0, value=1.5}
         {interpMode=RCIM_Cubic, tangentMode=RCTM_Break, time=0.9, value=3}
        ]}}}
     Scale Curve=0.1
    )
    Index 0 Param 3=&FloatFromCurve(
     FloatCurve=DataInterface() {propertyValues {Curve {keys [
         {interpMode=RCIM_Cubic, leaveTangentWeight=0.9958, tangentMode=RCTM_User, time=0, value=0.4}
         {arriveTangent=-0.6059, interpMode=RCIM_Cubic, tangentMode=RCTM_Break, time=1, value=1}
        ]}}}
     Scale Curve=0.75
    )
    Index 0 Param 4=500
    Module.Param0WriteEnabled=true
    Module.Param1WriteEnabled=true
    Module.Param2WriteEnabled=true
    Module.Param3WriteEnabled=true
   )
   &ScaleColor(
    Scale Alpha=&MaskFloatBySpawnGroup(
     Masked Float=&FloatFromCurve(
      FloatCurve=DataInterface() {propertyValues {Curve {keys [
          {interpMode=RCIM_Cubic, leaveTangentWeight=1.182, tangentMode=RCTM_User, time=0, value=0}
          <at 0.5 1 RCIM_Cubic>
          {arriveTangent=-5.747, interpMode=RCIM_Cubic, tangentMode=RCTM_Break, time=1, value=0}
         ]}}}
      Scale Curve=&Multiply_Float(
       A=0.5
       B=&FN_LOD_Distance_Remap(Start Distance=1000, End Distance=2500, Minimum Output=0, Maximum Output=1, Clamp Output=true)
      )
     )
     Passthrough Float=&MaskFloatBySpawnGroup(
      Masked Float=&FloatFromCurve(
       FloatCurve=DataInterface() {propertyValues {Curve {keys [
           <at 0 0 RCIM_Cubic>
           {arriveTangent=1.4286, interpMode=RCIM_Cubic, leaveTangent=1.4286, time=0.2, value=1}
           {arriveTangent=-1.25, interpMode=RCIM_Cubic, leaveTangent=-1.25, time=0.7, value=1}
           <at 1 0>
          ]}}}
       Scale Curve=&Multiply_Float(
        A=0.5
        B=&FN_LOD_Distance_Remap(Start Distance=1000, End Distance=2500, Minimum Output=0, Maximum Output=1, Clamp Output=true)
       )
      )
      Passthrough Float=&MaskFloatBySpawnGroup(
       Masked Float=&FloatFromCurve(
        FloatCurve=DataInterface() {propertyValues {Curve {keys [
            {interpMode=RCIM_Cubic, tangentMode=RCTM_User, time=0, value=0}
            {interpMode=RCIM_Cubic, tangentMode=RCTM_User, time=0.5, value=1}
            {tangentMode=RCTM_User, time=1, value=0}
           ]}}}
        Scale Curve=0.25
       )
       Passthrough Float=&MaskFloatBySpawnGroup(
        Masked Float=&FloatFromCurve(
         FloatCurve=DataInterface() {propertyValues {Curve {keys [
             {arriveTangent=12.4054, interpMode=RCIM_Cubic, leaveTangent=12.4054, tangentMode=RCTM_User, time=0.125, value=0}
             <at 0.2 1 RCIM_Cubic>
             {arriveTangent=-0.7128, interpMode=RCIM_Cubic, leaveTangent=-0.7128, tangentMode=RCTM_User, time=1, value=0}
            ]}}}
         Scale Curve=&Multiply_Float(
          A=&RandomRangeFloat(Minimum=0.5, Maximum=1)
          B=&FN_LOD_Distance_Remap(Start Distance=500, End Distance=2500, Minimum Output=0, Maximum Output=1, Clamp Output=true)
         )
        )
        Passthrough Float=&MaskFloatBySpawnGroup(
         Masked Float=&FloatFromCurve(FloatCurve=DataInterface() {propertyValues {Curve {keys [
              {interpMode=RCIM_Cubic, tangentMode=RCTM_Break, time=0, value=0}
              {interpMode=RCIM_Cubic, tangentMode=RCTM_Break, time=0.16, value=0.8}
              {interpMode=RCIM_Cubic, tangentMode=RCTM_Break, time=1, value=0}
             ]}}})
         Passthrough Float=&MaskFloatBySpawnGroup(
          Masked Float=&FloatFromCurve(
           FloatCurve=DataInterface() {propertyValues {Curve {keys [{arriveTangent=1.0833, interpMode=RCIM_Cubic, time=0, value=0}, <at 0.5024 0.6194 RCIM_Cubic>, <at 1 0>]}}}
           Scale Curve=&Multiply_Float(
            A=1
            B=&FN_LOD_Distance_Remap(Start Distance=500, End Distance=2500, Minimum Output=0, Maximum Output=1, Clamp Output=true)
           )
          )
          Passthrough Float=&Multiply_Float(
           A=&FloatFromCurve(
            FloatCurve=DataInterface() {propertyValues {Curve {keys [
                <at 0 0 RCIM_Cubic>
                {interpMode=RCIM_Cubic, tangentMode=RCTM_User, time=0.1, value=0.9914}
                {interpMode=RCIM_Cubic, tangentMode=RCTM_User, time=0.7, value=1}
                <at 1 0>
               ]}}}
            Scale Curve=0.25
           )
           B=&Waveform(
            "Waveform [1]"=::NewEnumerator0
            "[1] Frequency"=&RandomRangeFloat(Minimum=5, Maximum=8)
            "[1] Amplitude Min/Max"=<0.6 1>
           )
          )
          Spawn Group Mask=5
         )
         Spawn Group Mask=4
        )
        Spawn Group Mask=3
       )
       Spawn Group Mask=2
      )
      Spawn Group Mask=1
     )
    )
    Scale RGB=&MaskVectorBySpawnGroup(
     Masked Vector=/Niagara/DynamicInputs/TypeConversions/MakeVectorFromLinearColorHSV(Color:LinearColor=&HueShiftLinearColor(
       Hue=&FloatFromCurve(FloatCurve=DataInterface() {propertyValues {Curve {keys [<at 0 0>,<at 1 0.25>]}}})
       Saturation=1
       Value=1
       Linear Color=$Particles.Initial.Color
      ))
     Passthrough Vector=&MaskVectorBySpawnGroup(
      Masked Vector=<5 5 5>
      Passthrough Vector=&MaskVectorBySpawnGroup(
       Masked Vector=<1 1 1>
       Passthrough Vector=&MaskVectorBySpawnGroup(
        Masked Vector=<1 1 1>
        Passthrough Vector=&MaskVectorBySpawnGroup(
         Masked Vector=<1 1 1>
         Passthrough Vector=&MaskVectorBySpawnGroup(
          Masked Vector=&MakeVectorFromLinearColorRGB(Color=&ColorFromCurve(ColorCurve=DataInterface() {propertyValues {
              AlphaCurve {keys [<at 0 1 RCIM_Cubic>,<at 1 0 RCIM_Cubic>]}
              BlueCurve {keys [<at 0.2176 1>, {arriveTangent=-0.1731, interpMode=RCIM_Cubic, time=0.5937, value=1}, <at 1 0.8281 RCIM_Cubic>]}
              GreenCurve {keys [<at 0.2176 1>, {arriveTangent=-0.8557, interpMode=RCIM_Cubic, time=0.5937, value=0.646}, <at 1 0.1504 RCIM_Cubic>]}
              RedCurve {keys [<at 0.2176 1>,<at 0.5937 0.1136 RCIM_Cubic>,<at 1 1 RCIM_Cubic>]}
             }}))
          Passthrough Vector=&MakeVectorFromLinearColorRGB(Color=&ColorFromCurve(ColorCurve=DataInterface() {propertyValues {
              AlphaCurve {keys [<at 0 1 RCIM_Cubic>,<at 1 0 RCIM_Cubic>]}
              BlueCurve {keys [<at 0 1 RCIM_Cubic>,<at 1 1 RCIM_Cubic>]}
              GreenCurve {keys [<at 0 0.6875 RCIM_Cubic>,<at 1 0.0414 RCIM_Cubic>]}
              RedCurve {keys [<at 0 0.1067 RCIM_Cubic>,<at 1 0.6941 RCIM_Cubic>]}
             }}))
          Spawn Group Mask=5
         )
         Spawn Group Mask=4
        )
        Spawn Group Mask=3
       )
       Spawn Group Mask=2
      )
      Spawn Group Mask=1
     )
    )
    ScaleA=true
    ScaleRGB=true
   )
   &UpdateMeshOrientation(
    Rotation Rate=&MaskFloatBySpawnGroup(
     Masked Float=0
     Passthrough Float=&MaskFloatBySpawnGroup(
      Masked Float=0.5
      Passthrough Float=&MaskFloatBySpawnGroup(
       Masked Float=0
       Passthrough Float=&MaskFloatBySpawnGroup(Masked Float=0, Passthrough Float=0, Spawn Group Mask=3)
       Spawn Group Mask=2
      )
      Spawn Group Mask=1
     )
    )
    Rotation Vector=&RandomVector(Vector Scale=1)
   ) {module=UpdateMeshOrientation001}
   &ScaleMeshSize(Scale Factor=&VectorFromFloat(Value=&FloatFromCurve(FloatCurve=DataInterface() {propertyValues {Curve {keys [<at 2000 0>,<at 3000 1>]}}}, CurveIndex=$Engine.Owner.LODDistance))) {module=ScaleMeshSize001}
   &SpriteRotationRate(Rotation Rate=&RandomRangeFloat(Minimum=-20, Maximum=20))
  ]
  rendererClasses [&NiagaraMeshRendererProperties]
  renderers [{rendererClass=&NiagaraMeshRendererProperties, rendererIndex=0}]
 }
 Plasma_Flow {
  emitterUpdateScript [
   &EmitterState(Life Cycle Mode=::NewEnumerator1, Loop Behavior=::NewEnumerator1, Loop Duration Mode=::NewEnumerator1, UseLoopDelay=false)
   &SpawnBurst_Instantaneous(Spawn Count=2, Use Spawn Probability=false)
  ]
  particleSpawnScript [
   &FN_SetSpawnGroup(Number of Spawn Groups=2)
   &InitializeParticle(
    Color=&MakeCustomLinearColorFromBool(
     Bool=&SetBoolByIntComparison(A=$Particles.SpawnGroup, B=0, Comparison Type=::NewEnumerator4)
     True Color=<rgba 0.059 0.3226 1 0.03>
     False Color=<rgba 0.35 0 1 0.06>
    )
    Mesh Scale Mode=::NewEnumerator3
    Mesh Uniform Scale=&MakeCustomFloatFromBool(
     Bool=&SetBoolByIntComparison(A=$Particles.SpawnGroup, B=0, Comparison Type=::NewEnumerator4)
     True Float=11
     False Float=12
    )
    UsePositionOffset=false
    Write Material Random=true
   )
   &InitialMeshOrientation(Orientation Axis=<0 0 1>, Orientation Vector=<0 0 1>, Rotation=<0 0.25 0>, Use Orientation Vector=true, Use Rotation Vector=true) {module=InitialMeshOrientation001}
   &InitialMeshOrientation(
    Rotation=&MakeCustomVectorFromBool(
     Bool=&SetBoolByIntComparison(A=$Particles.SpawnGroup, B=0, Comparison Type=::NewEnumerator4)
     True Vector=<0 -0.05 -0.05>
     False Vector=<0.5 0.05 0.05>
    )
    Use Orientation Vector=true
    Use Rotation Vector=true
   )
   &SystemLocation(MaskSpawnGroup=false, Offset=<0 0 -120>)
  ]
  particleUpdateScript [
   &ParticleState(Kill Particles When Lifetime Has Elapsed=false, Let Infinitely Lived Particles Die When Emitter Deactivates=true)
   &ScaleColor(
    Scale Alpha=&Multiply_Float(
     A=$System.DistanceFade
     B=&FNTimeOfDayScalar(Night Scale=0.7, Morning Scale=0.5, Evening Scale=0.5)
    )
    ScaleA=true
    ScaleRGB=true
   )
   &CameraOffset(Camera Offset Amount=250)
  ]
  rendererClasses [&NiagaraMeshRendererProperties]
  renderers [
   {rendererClass=&NiagaraMeshRendererProperties, rendererIndex=0}
   {rendererClass=&NiagaraMeshRendererProperties, rendererIndex=1}
  ]
 }
 Radial_Orb {
  emitterUpdateScript [
   &EmitterState(
    Inactive Response=::NewEnumerator1
    Life Cycle Mode=::NewEnumerator1
    Loop Behavior=::NewEnumerator1
    Loop Duration Mode=::NewEnumerator1
    UseLoopDelay=false
   )
   &SpawnBurst_Instantaneous(Spawn Group=1, Use Spawn Probability=false, UseLoopCountLimit=false) {module=SpawnBurst_Instantaneous001}
  ]
  particleSpawnScript [
   /SetVariables(Particles.VisibilityTag=&SpawnGroup()) {isSetParameters=true}
   &InitializeParticle(
    Color=<rgba 0.1623 0.1588 0.156 0.5>
    Mesh Scale Mode=::NewEnumerator3
    Mesh Uniform Scale=38
    Position Offset=$System.FXPosition
    Sprite Size Mode=::NewEnumerator3
    Uniform Sprite Size=200
    UsePositionOffset=false
    Write Material Random=true
   )
   &InitialMeshOrientation(Mesh Orientation Mode=::NewEnumerator2, Rotation=<0 0 0.25>, Use Rotation Vector=true)
   &SystemLocation(MaskSpawnGroup=false, Offset=<0 0 -155>)
  ]
  particleUpdateScript [
   &ParticleState(Kill Particles When Lifetime Has Elapsed=false, Let Infinitely Lived Particles Die When Emitter Deactivates=true)
   &UpdateMeshOrientation(Rotation Rate=0.1, Rotation Vector=<0 1 0>)
   &ScaleMeshSize(Scale Factor=&VectorFromFloat(Value=&Sine(
      Normalized Angle=$Emitter.Age
      Period=6
      Scale=0.1
      Bias=&Sine(Normalized Angle=$Emitter.Age, Period=0.4, Scale=0.004, Bias=1)
     )))
   &DynamicMaterialParameters(
    Index 0 Param 2=0.5
    Module.Index0Float0Write=true
    Module.Index0Float1Write=true
    Module.Index0Float2Write=true
    Module.Index0Float3Write=true
   )
   &CameraOffset(Camera Offset Amount=-1000) {module=CameraOffset001}
   &ScaleColor(
    Scale Alpha=&Multiply_Float(
     A=&FNTimeOfDayScalar(Night Scale=1.1, Morning Scale=1.3, Day Scale=1.35, Evening Scale=1.2)
     B=$System.DistanceFade
    )
    ScaleA=true
    ScaleRGB=true
   )
  ]
  rendererClasses [&NiagaraMeshRendererProperties]
  renderers [{rendererClass=&NiagaraMeshRendererProperties, rendererIndex=0}]
 }
 Rays {
  bEnabled=false
  emitterUpdateScript [
   &EmitterState()
   &SpawnRate(Spawn Probability=0.75, SpawnGroup=1, SpawnRate=3, Use Spawn Probability=true)
  ]
  particleSpawnScript [
   &InitializeParticle(
    Color=&FNTimeOfDayColor(
     Night Color=<rgba 0.1403 0.0968 0.7 1>
     Morning Color=<rgba 0.6468 0 2 1>
     Day Color=<rgba 0.7065 0.004 2.5 1>
     Evening Color=<rgba 0.7629 0.0155 2 1>
    )
    Lifetime Max=3
    Lifetime Min=2.5
    Lifetime Mode=::NewEnumerator1
    Sprite Size Mode=::NewEnumerator1
    Uniform Sprite Size Max=2200
    Uniform Sprite Size Min=2000
    UsePositionOffset=false
    Write Material Random=true
   )
   &SystemLocation(MaskSpawnGroup=false, Offset=<0 0 -120>)
  ]
  particleUpdateScript [
   &ParticleState()
   &FN_Radial_Material(
    Base Sprite Offset X:NiagaraFloat=&RandomRangeFloat(Minimum=0, Maximum=1)
    Base Sprite Offset Y:NiagaraFloat=&RandomRangeFloat(Minimum=0, Maximum=1)
    Base Sprite Tile U=0.2
    Base Sprite Tile V=3
    Depth Fade=2500
    Radial Mask Power:NiagaraFloat=&FloatFromCurve(
     FloatCurve=DataInterface() {propertyValues {Curve {keys [{tangentMode=RCTM_User, time=0, value=0.5}, {tangentMode=RCTM_User, time=0.9, value=5}]}}}
     Scale Curve=4
    )
    Radial Mask Strength:NiagaraFloat=&FloatFromCurve(FloatCurve=DataInterface() {propertyValues {Curve {keys [
         {interpMode=RCIM_Cubic, tangentMode=RCTM_User, time=0, value=1}
         {arriveTangent=-2.8883, interpMode=RCIM_Cubic, leaveTangent=-2.8883, tangentMode=RCTM_User, time=1, value=0}
        ]}}})
    Sprite Extrusion Speed=-1
    Sprite Spin Speed:NiagaraFloat=&RandomRangeFloat(Minimum=0, Maximum=0)
   )
   &ScaleSpriteSize(Scale Factor=&Vector2DFromFloat(Value=&FloatFromCurve(FloatCurve=DataInterface() {propertyValues {Curve {keys [
          {arriveTangent=0.6574, interpMode=RCIM_Cubic, leaveTangent=0.6574, tangentMode=RCTM_User, time=0, value=0.7}
          {interpMode=RCIM_Cubic, tangentMode=RCTM_User, time=1, value=1}
         ]}}})))
   &ScaleColor(
    Scale Alpha=&FloatFromCurve(
     FloatCurve=DataInterface() {propertyValues {Curve {keys [
         {arriveTangent=9.0593, interpMode=RCIM_Cubic, leaveTangent=9.0593, tangentMode=RCTM_User, time=0, value=-0.0196}
         {arriveTangent=0.0196, interpMode=RCIM_Cubic, leaveTangent=0.0196, time=0.2923, value=1.0741}
         {arriveTangent=-1.1678, interpMode=RCIM_Cubic, leaveTangent=-1.1678, tangentMode=RCTM_User, time=1, value=0}
        ]}}}
     Scale Curve=&FloatFromCurve(
      FloatCurve=DataInterface() {propertyValues {Curve {keys [<at 0 1>, {arriveTangent=-0.0001, interpMode=RCIM_Cubic, leaveTangent=-0.0001, time=4000, value=1}, <at 9500 0 RCIM_Cubic>]}}}
      CurveIndex=$Engine.Owner.LODDistance
     )
    )
    Scale RGB=&RandomRangeVector(Minimum=<0.8 0.8 0.8>, Maximum=<1 1 1>)
    ScaleA=true
    ScaleRGB=true
   ) {module=ScaleColor002}
  ]
  rendererClasses [&NiagaraSpriteRendererProperties]
  renderers [{rendererClass=&NiagaraSpriteRendererProperties}]
 }
 Rays001 {
  emitterUpdateScript [
   &EmitterState(Life Cycle Mode=::NewEnumerator1, Loop Behavior=::NewEnumerator1, Loop Duration Mode=::NewEnumerator1, UseLoopDelay=false)
   &SpawnBurst_Instantaneous(Use Spawn Probability=false, UseLoopCountLimit=false)
  ]
  particleSpawnScript [
   &InitializeParticle(
    Color=&FNTimeOfDayColor(
     Night Color=<rgba 0.1403 0.0968 0.7 1>
     Morning Color=<rgba 0.6468 0 2 1>
     Day Color=<rgba 0.7065 0.004 2.5 1>
     Evening Color=<rgba 0.7629 0.0155 2 1>
    )
    Lifetime Max=3
    Lifetime Min=2.5
    Lifetime Mode=::NewEnumerator1
    Sprite Size Mode=::NewEnumerator1
    Uniform Sprite Size Max=2200
    Uniform Sprite Size Min=2000
    UsePositionOffset=false
    Write Material Random=true
   )
   &SystemLocation(MaskSpawnGroup=false, Offset=<0 0 -120>)
  ]
  particleUpdateScript [
   &ParticleState(
    Kill Particles When Lifetime Has Elapsed=false
    Let Infinitely Lived Particles Die When Emitter Deactivates=true
    Loop Particles Lifetime=true
   )
   &FN_Radial_Material(
    Base Sprite Offset X:NiagaraFloat=&RandomRangeFloat(Minimum=0, Maximum=1)
    Base Sprite Offset Y:NiagaraFloat=&RandomRangeFloat(Minimum=0, Maximum=1)
    Base Sprite Tile U=0.2
    Base Sprite Tile V=3
    Depth Fade=2500
    Radial Mask Power=7
    Sprite Extrusion Speed=-1
    Sprite Spin Speed:NiagaraFloat=&RandomRangeFloat(Minimum=0, Maximum=0)
   ) {module=FN_Radial_Material002}
   &ScaleColor(
    Scale Alpha=&FloatFromCurve(
     FloatCurve=DataInterface() {propertyValues {Curve {keys [<at 0 1>, {arriveTangent=-0.0001, interpMode=RCIM_Cubic, leaveTangent=-0.0001, time=4000, value=1}, <at 9500 0 RCIM_Cubic>]}}}
     CurveIndex=$Engine.Owner.LODDistance
    )
    Scale RGB=&RandomRangeVector(Minimum=<0.8 0.8 0.8>, Maximum=<1 1 1>)
    ScaleA=false
    ScaleRGB=true
   )
  ]
  rendererClasses [&NiagaraSpriteRendererProperties]
  renderers [{rendererClass=&NiagaraSpriteRendererProperties}]
 }
 Splines001 {
  emitterUpdateScript [
   &EmitterState()
   &SpawnRate(Use Spawn Probability=false)
  ]
  particleSpawnScript [
   &InitializeParticle(
    Color Maximum=<rgba 0.9393 0.5491 1 0.7>
    Color Minimum=<rgba 1 1 1 1>
    Color Mode=::NewEnumerator2
    Lifetime Max=8
    Lifetime Min=6
    Lifetime Mode=::NewEnumerator1
    Mesh Scale Max=<15 30 30>
    Mesh Scale Min=<10 10 10>
    Mesh Scale Mode=::NewEnumerator2
    UsePositionOffset=false
    Write Material Random=true
   )
   &ShapeLocation(
    Hemisphere Distribution=<1 0.42>
    MaskSpawnGroup=false
    Override Seed=false
    Sphere Radius=10
    Sphere Surface Distribution=1
    UseOwnerScale=true
   )
   &InitialMeshOrientation(
    Mesh Orientation Mode=::NewEnumerator2
    Rotation=&RandomRangeVector(Minimum=<0 0 0>, Maximum=<1 0 0>)
    Use Rotation Vector=true
   )
   &AddVelocity(Module.Use Falloff Distance=false, Override Seed=false, Velocity Mode=::NewEnumerator1, Velocity Speed=50)
  ]
  particleUpdateScript [
   &ParticleState()
   &ScaleMeshSize(Scale Factor=&VectorFromCurve(VectorCurve=DataInterface() {propertyValues {
       XCurve {
        defaultValue=1
        keys [
         {arriveTangent=0.3265, interpMode=RCIM_Cubic, leaveTangent=0.3265, tangentMode=RCTM_User, time=0, value=0.2}
         {arriveTangent=0.4285, interpMode=RCIM_Cubic, leaveTangent=0.4285, tangentMode=RCTM_User, time=1, value=0.5}
        ]
       }
       YCurve {
        defaultValue=1
        keys [
         {arriveTangent=1.3531, interpMode=RCIM_Cubic, leaveTangent=1.3531, tangentMode=RCTM_User, time=0, value=0.25}
         {arriveTangent=0.3205, interpMode=RCIM_Cubic, leaveTangent=0.3205, tangentMode=RCTM_User, time=1, value=1}
        ]
       }
       ZCurve {
        defaultValue=1
        keys [
         {arriveTangent=1.3531, interpMode=RCIM_Cubic, leaveTangent=1.3531, tangentMode=RCTM_User, time=0, value=0.25}
         {arriveTangent=0.3205, interpMode=RCIM_Cubic, leaveTangent=0.3205, tangentMode=RCTM_User, time=1, value=1}
        ]
       }
      }})) {module=ScaleMeshSize001}
   &SolveForcesAndVelocity(Clamp Velocity=false, Limit Acceleration=false)
   &DynamicMaterialParameters(
    Index 0 Param 1=&RandomRangeFloat(Minimum=0, Maximum=1)
    Index 0 Param 2=&RandomRangeFloat(Minimum=1, Maximum=3)
    Index 0 Param 4=&FloatFromCurve(
     FloatCurve=DataInterface() {propertyValues {Curve {defaultValue=1, keys [{interpMode=RCIM_Cubic, tangentMode=RCTM_User, time=0.25, value=0}, <at 1 1 RCIM_Cubic>]}}}
     Scale Curve=3
    )
    Module.Index0Float0Write=true
    Module.Index0Float1Write=true
    Module.Index0Float2Write=true
    Module.Index0Float3Write=true
   )
   &ScaleColor(
    Scale Alpha=&FloatFromCurve(
     FloatCurve=DataInterface() {propertyValues {Curve {
        defaultValue=1
        keys [
         {arriveTangent=6.4544, interpMode=RCIM_Cubic, time=0, value=0}
         {
          arriveTangent=1.5375
          arriveTangentWeight=0.1333
          interpMode=RCIM_Cubic
          leaveTangent=1.5375
          leaveTangentWeight=0.0667
          tangentMode=RCTM_User
          time=0.3834
          value=0.449
         }
         {
          arriveTangent=-1.9657
          arriveTangentWeight=0.1296
          interpMode=RCIM_Cubic
          leaveTangent=-1.9657
          leaveTangentWeight=0.2592
          tangentMode=RCTM_User
          time=0.6793
          value=0.8725
         }
         {tangentMode=RCTM_User, time=1, value=0}
        ]
       }}}
     Scale Curve=&FloatFromCurve(
      FloatCurve=DataInterface() {propertyValues {Curve {keys [<at 3000 0>, <at 5000 1>, {arriveTangent=-0.0001, interpMode=RCIM_Cubic, leaveTangent=-0.0001, time=14500, value=1}, <at 17500 0 RCIM_Cubic>]}}}
      CurveIndex=$Engine.Owner.LODDistance
     )
    )
    ScaleA=true
    ScaleRGB=false
   )
  ]
  rendererClasses [&NiagaraMeshRendererProperties]
  renderers [{rendererClass=&NiagaraMeshRendererProperties, rendererIndex=0}]
 }
}
summary {
 systemName=NS_Env_TechHQ_ShardPowered_Heart_Details
 emitters [
  {emitterName=Splines001, rendererClasses [&NiagaraMeshRendererProperties]}
  {emitterName=Energy, rendererClasses [&NiagaraMeshRendererProperties]}
  {emitterName=Radial_Orb, rendererClasses [&NiagaraMeshRendererProperties]}
  {emitterName=Plasma_Flow, rendererClasses [&NiagaraMeshRendererProperties]}
  {bEnabled=false, emitterName=Rays, rendererClasses [&NiagaraSpriteRendererProperties]}
  {emitterName=Debris, rendererClasses [&NiagaraMeshRendererProperties]}
  {emitterName=Central_Flare, rendererClasses []}
  {emitterName=Rays001, rendererClasses [&NiagaraSpriteRendererProperties]}
 ]
 userVariables [
  User.IsClosed {type=NiagaraBool, defaultValue=true}
  User.IsOpen {type=NiagaraBool, defaultValue=true}
 ]
}
