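# Niagara system for the TechHQ shard environment effect: purple-blue energy made of meshes, sprites and a glowing light.
# The Energy emitter packs several sub-effects into spawn groups, using per-group masked inputs, a VisibilityTag and per-effect dynamic material vectors across its renderers.
# User.Stage sets the glow-burst mesh scale, and camera distance modulates several alphas.
# A distance-culled Light emitter adds glow, while standalone versions like Motes, Mesh_Swirls and TinyFrags are disabled.
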NiagaraSystem /Game/Effects/Fort_Effects/Effects/Environment/Hera/TechHQ/NS_TechHQ_Shard
dependencies {
 usedDataInterfaces [&NiagaraDataInterfaceArrayColor, &NiagaraDataInterfaceCurve, &NiagaraDataInterfaceColorCurve]
 usedDynamicInputs []
 usedModules [
  &SystemState
  &self:SystemSpawnScript.SystemScriptSource.SystemScriptGraph.NiagaraNodeAssignment_0.SetVariables_#1
 ]
}
emitters {
 Energy {
  emitterUpdateScript [
   /SetVariables(Emitter.PanSpeed=0.5) {isSetParameters=true}
   &EmitterState()
   &SpawnRate(Use Spawn Probability=false)
   &SpawnRate(SpawnGroup=1, SpawnRate=20, Use Spawn Probability=false) {module=SpawnRate001}
   &SpawnRate(SpawnGroup=2, SpawnRate=2, Use Spawn Probability=false) {module=SpawnRate002}
   &SpawnRate(SpawnGroup=3, SpawnRate=10, Use Spawn Probability=false) {module=SpawnRate003}
   &SpawnRate(SpawnGroup=4, SpawnRate=2, Use Spawn Probability=false) {module=SpawnRate004}
   &SpawnRate(SpawnGroup=5, SpawnRate=2, Use Spawn Probability=false) {module=SpawnRate005}
   &SpawnRate(SpawnGroup=6, SpawnRate=30, Use Spawn Probability=false) {module=SpawnRate006}
  ]
  particleSpawnScript [
   /SetVariables(Particles.VisibilityTag=&SpawnGroup()) {isSetParameters=true}
   &InitializeParticle(
    Color=&MaskLinearColorBySpawnGroup(
     Masked Linear Color=&RandomRangeLinearColor(Minimum=<rgba 2.5648 0.569 5 1>, Maximum=<rgba 5 1.0977 4.2338 1>)
     Passthrough Linear Color=&MaskLinearColorBySpawnGroup(
      Masked Linear Color=&RandomRangeLinearColor(Minimum=<rgba 1 0.0336 0.9623 1>, Maximum=<rgba 0.7292 0.745 1 1>)
      Passthrough Linear Color=&MaskLinearColorBySpawnGroup(
       Masked Linear Color=<rgba 1.2844 1.0153 4 0.75>
       Passthrough Linear Color=&MaskLinearColorBySpawnGroup(
        Masked Linear Color=&SelectLinearColorFromArray(
         Array Sampling Mode=::NewEnumerator1
         Direct Array Index=&RandomRangeInt(Minimum=0, Maximum=1)
         Color Selection Array=DataInterface() {propertyValues {ColorData [<rgba 0.3199 0.5916 1 1>, <rgba 0.9443 0.3385 1 1>]}}
         Random Seed=?
        )
        Passthrough Linear Color=&MaskLinearColorBySpawnGroup(
         Masked Linear Color=<rgba 0.1907 0 1 0.25>
         Passthrough Linear Color=&MaskLinearColorBySpawnGroup(Masked Linear Color=<rgba 1 1 1 1>, Passthrough Linear Color=<rgba 1 1 1 1>, Spawn Group Mask=5)
         Spawn Group Mask=4
        )
        Spawn Group Mask=3
       )
       Spawn Group Mask=2
      )
      Spawn Group Mask=1
     )
    )
    Lifetime=&MaskFloatBySpawnGroup(
     Masked Float=2
     Passthrough Float=&MaskFloatBySpawnGroup(
      Masked Float=&RandomRangeFloat(Minimum=0.5, Maximum=1)
      Passthrough Float=&MaskFloatBySpawnGroup(
       Masked Float=&RandomRangeFloat(Minimum=2, Maximum=3)
       Passthrough Float=&MaskFloatBySpawnGroup(
        Masked Float=&RandomRangeFloat(Minimum=3, Maximum=4)
        Passthrough Float=&MaskFloatBySpawnGroup(
         Masked Float=4
         Passthrough Float=&MaskFloatBySpawnGroup(Masked Float=2, Passthrough Float=&RandomRangeFloat(Minimum=1, Maximum=2), Spawn Group Mask=5)
         Spawn Group Mask=4
        )
        Spawn Group Mask=3
       )
       Spawn Group Mask=2
      )
      Spawn Group Mask=1
     )
    )
    Mesh Scale=&MaskVectorBySpawnGroup(
     Masked Vector=<0.35 0.35 0.35>
     Passthrough Vector=&MaskVectorBySpawnGroup(
      Masked Vector=&RandomRangeVector(Minimum=<0.4 0.4 1>, Maximum=<0.7 0.7 1>, Evaluation Type=::NewEnumerator1)
      Passthrough Vector=&MaskVectorBySpawnGroup(
       Masked Vector=&VectorFromFloat(Value=&self:StageSelection_02(
         Stage:NiagaraInt32=$User.Stage
         Stage 0:NiagaraFloat=0
         Stage 1:NiagaraFloat=0.9
         Stage 2:NiagaraFloat=1.2
         Stage 3:NiagaraFloat=1.5
         Stage 4:NiagaraFloat=2.2
         Stage 5:NiagaraFloat=2.5
         Stage 6:NiagaraFloat=4
        ))
       Passthrough Vector=&MaskVectorBySpawnGroup(
        Masked Vector=&RandomRangeVector(Minimum=<2 1.5 1.5>, Maximum=<3 8 8>)
        Passthrough Vector=<8 8 8>
        Spawn Group Mask=3
       )
       Spawn Group Mask=2
      )
      Spawn Group Mask=1
     )
    )
    Mesh Scale Mode=::NewEnumerator4
    Sprite Size Max=<125 125>
    Sprite Size Min=<85 85>
    Sprite Size Mode=::NewEnumerator2
    UsePositionOffset=false
    Write Material Random=true
   )
   /SetVariables(Particles.Sparkle:NiagaraFloat=&UniformAOrBFloat(A=0, Distribution Weight=0.2, Evaluation Type=::NewEnumerator0, Recalculate Random Each Loop=true, Override Seed=false)) {isSetParameters=true, module=SetVariables_#2}
   &InitialMeshOrientation(
    Mesh Orientation Mode=::NewEnumerator2
    Orientation Vector=&MaskVectorBySpawnGroup(
     Masked Vector=<0 0 1>
     Passthrough Vector=&MaskVectorBySpawnGroup(
      Masked Vector=&RandomVector(Vector Scale=1)
      Passthrough Vector=&MaskVectorBySpawnGroup(
       Masked Vector=&RandomVector(Vector Scale=1)
       Passthrough Vector=&MaskVectorBySpawnGroup(Masked Vector=<1 0 0>, Passthrough Vector=<0 0 0>, Spawn Group Mask=3)
       Spawn Group Mask=2
      )
      Spawn Group Mask=1
     )
    )
    Rotation=&MaskVectorBySpawnGroup(
     Masked Vector=<0 0 0>
     Passthrough Vector=&MaskVectorBySpawnGroup(
      Masked Vector=<0 0 0>
      Passthrough Vector=&MaskVectorBySpawnGroup(
       Masked Vector=<0 0 0>
       Passthrough Vector=&MaskVectorBySpawnGroup(
        Masked Vector=&RandomRangeVector(Minimum=<-1 -1 -1>, Maximum=<1 1 1>)
        Passthrough Vector=&RandomRangeVector(Minimum=<-0.7 0.25 0>, Maximum=<0.7 0.5 0>)
        Spawn Group Mask=3
       )
       Spawn Group Mask=2
      )
      Spawn Group Mask=1
     )
    )
    Use Rotation Vector=true
   )
   &ShapeLocation(
    MaskSpawnGroup=false
    Offset Mode=::NewEnumerator0
    Override Seed=false
    Sphere Radius=&MaskFloatBySpawnGroup(
     Masked Float=0
     Passthrough Float=&MaskFloatBySpawnGroup(
      Masked Float=400
      Passthrough Float=&MaskFloatBySpawnGroup(
       Masked Float=0
       Passthrough Float=&MaskFloatBySpawnGroup(
        Masked Float=0
        Passthrough Float=&MaskFloatBySpawnGroup(
         Masked Float=0
         Passthrough Float=&MaskFloatBySpawnGroup(Masked Float=0, Passthrough Float=400, Spawn Group Mask=5)
         Spawn Group Mask=4
        )
        Spawn Group Mask=3
       )
       Spawn Group Mask=2
      )
      Spawn Group Mask=1
     )
    )
    UseOwnerScale=true
   )
   &AddVelocity(
    Module.Use Falloff Distance=false
    Override Seed=false
    Velocity Mode=::NewEnumerator1
    Velocity Speed=&MaskFloatBySpawnGroup(
     Masked Float=0
     Passthrough Float=&MaskFloatBySpawnGroup(
      Masked Float=&RandomRangeFloat(Minimum=250, Maximum=500)
      Passthrough Float=&MaskFloatBySpawnGroup(
       Masked Float=0
       Passthrough Float=&MaskFloatBySpawnGroup(
        Masked Float=0
        Passthrough Float=&MaskFloatBySpawnGroup(
         Masked Float=0
         Passthrough Float=&MaskFloatBySpawnGroup(Masked Float=0, Passthrough Float=250, Spawn Group Mask=5)
         Spawn Group Mask=4
        )
        Spawn Group Mask=3
       )
       Spawn Group Mask=2
      )
      Spawn Group Mask=1
     )
    )
   ) {module=AddVelocity001}
   &AddVelocity(Velocity=&MaskVectorBySpawnGroup(
     Masked Vector=<0 0 0>
     Passthrough Vector=&MaskVectorBySpawnGroup(
      Masked Vector=<0 0 0>
      Passthrough Vector=&MaskVectorBySpawnGroup(
       Masked Vector=<0 0 0>
       Passthrough Vector=&MaskVectorBySpawnGroup(
        Masked Vector=<0 0 0>
        Passthrough Vector=&MaskVectorBySpawnGroup(
         Masked Vector=<0 0 0>
         Passthrough Vector=&RandomRangeVector(Minimum=<-100 -100 -100>, Maximum=<100 100 100>)
         Spawn Group Mask=4
        )
        Spawn Group Mask=3
       )
       Spawn Group Mask=2
      )
      Spawn Group Mask=1
     )
    ))
   &ApplyInitialForces()
   /SetVariables(Particles.Shimmers_Dynamic:Vector4f=&MakeVector4(
     X=&RandomRangeFloat(Minimum=0, Maximum=1, Evaluation Type=::NewEnumerator1)
     Y=&RandomRangeFloat(Minimum=0.4, Maximum=0.6, Evaluation Type=::NewEnumerator1)
     Z=0.3
     W=0
    )) {isSetParameters=true}
  ]
  particleUpdateScript [
   &ParticleState()
   /SetVariables(Particles.TinyFrag_Dynamic:Vector4f=&MakeVector4(X=0, Y=0, Z=0, W=0)) {isSetParameters=true}
   /SetVariables(Particles.PurpleGlow_Burst_Dynamic:Vector4f=&MakeVector4(X=1.001, Y=1, Z=1, W=1)) {isSetParameters=true}
   /SetVariables(Particles.Swirl_Dynamic:Vector4f=&MakeVector4(
     X=&RandomRangeFloat(Minimum=0, Maximum=1)
     Y=&FloatFromCurve(FloatCurve=DataInterface() {propertyValues {Curve {
         defaultValue=1
         keys [
          {interpMode=RCIM_Cubic, tangentMode=RCTM_User, time=0, value=1}
          <at 0.5 0 RCIM_Cubic>
          {interpMode=RCIM_Cubic, tangentMode=RCTM_User, time=1, value=1}
         ]
        }}})
     Z=0.5
     W=&FloatFromCurve(
      FloatCurve=DataInterface() {propertyValues {Curve {keys [<at 0 0.3 RCIM_Cubic>, {arriveTangent=-3.2559, interpMode=RCIM_Cubic, leaveTangent=-3.2559, tangentMode=RCTM_User, time=1, value=-1}]}}}
      Scale Curve=0.5
     )
    )) {isSetParameters=true}
   /SetVariables(Particles.PlasmaWave_Dynamic:Vector4f=&MakeVector4(
     X=&FloatFromCurve(FloatCurve=DataInterface() {propertyValues {Curve {keys [<at 0 0.2 RCIM_Cubic>,<at 0.5 0>]}}})
     Y=&FloatFromCurve(FloatCurve=DataInterface() {propertyValues {Curve {keys [<at 0 1>,<at 0.5 0.75>]}}})
     Z=&FloatFromCurve(FloatCurve=DataInterface() {propertyValues {Curve {keys [<at 0 0>,<at 1 1>,<at 1 1>]}}})
     W=0
    )) {isSetParameters=true}
   /SetVariables(Particles.Motes_Dynamic:Vector4f=&MakeVector4(X=2, Y=1, Z=0.5, W=0)) {isSetParameters=true}
   /SetVariables(Particles.SparkleSize:Vector2f=/Niagara/DynamicInputs/Add/Add_Vector2(
     A:Vector2f=$Particles.SpriteSize
     B:Vector2f=&Vector2DFromFloat(Value=&Multiply_Float(
       A=$Particles.Sparkle
       B=&FloatFromCurve(
        FloatCurve=DataInterface() {propertyValues {Curve {keys [
            <at 0.4 0.05 RCIM_Cubic>
            {arriveTangent=-4.4497, interpMode=RCIM_Cubic, leaveTangent=-4.4497, tangentMode=RCTM_User, time=0.5, value=1}
            <at 1 0.05 RCIM_Cubic>
           ]}}}
        Scale Curve=350
       )
      ))
    )) {isSetParameters=true}
   &CurlNoiseForce(
    Noise Frequency=&MaskFloatBySpawnGroup(
     Masked Float=2
     Passthrough Float=&MaskFloatBySpawnGroup(
      Masked Float=0
      Passthrough Float=&MaskFloatBySpawnGroup(
       Masked Float=0
       Passthrough Float=&MaskFloatBySpawnGroup(
        Masked Float=0
        Passthrough Float=&MaskFloatBySpawnGroup(
         Masked Float=0
         Passthrough Float=&MaskFloatBySpawnGroup(Masked Float=0, Passthrough Float=5, Spawn Group Mask=5)
         Spawn Group Mask=4
        )
        Spawn Group Mask=3
       )
       Spawn Group Mask=2
      )
      Spawn Group Mask=1
     )
    )
    Noise Strength=&MaskFloatBySpawnGroup(
     Masked Float=0
     Passthrough Float=&MaskFloatBySpawnGroup(
      Masked Float=250
      Passthrough Float=&MaskFloatBySpawnGroup(
       Masked Float=0
       Passthrough Float=&MaskFloatBySpawnGroup(
        Masked Float=0
        Passthrough Float=&MaskFloatBySpawnGroup(
         Masked Float=0
         Passthrough Float=&MaskFloatBySpawnGroup(Masked Float=0, Passthrough Float=2500, Spawn Group Mask=5)
         Spawn Group Mask=4
        )
        Spawn Group Mask=3
       )
       Spawn Group Mask=2
      )
      Spawn Group Mask=1
     )
    )
    Pan Noise=false
   )
   &DynamicMaterialParameters(
    Index 0 Param 1=&FloatFromCurve(
     FloatCurve=DataInterface() {propertyValues {Curve {keys [<at 0 0.3 RCIM_Cubic>, {interpMode=RCIM_Cubic, tangentMode=RCTM_Break, time=0.9, value=1}]}}}
     Scale Curve=3
    )
    Index 0 Param 2=&FloatFromCurve(
     FloatCurve=DataInterface() {propertyValues {Curve {keys [
         {arriveTangent=-0.5351, interpMode=RCIM_Cubic, leaveTangent=-0.5351, tangentMode=RCTM_User, time=0, value=1.5}
         {interpMode=RCIM_Cubic, tangentMode=RCTM_Break, time=0.9, value=3}
        ]}}}
     Scale Curve=0.1
    )
    Index 0 Param 3=&FloatFromCurve(
     FloatCurve=DataInterface() {propertyValues {Curve {keys [
         {interpMode=RCIM_Cubic, leaveTangentWeight=0.9958, tangentMode=RCTM_User, time=0, value=0.4}
         {arriveTangent=-0.6059, interpMode=RCIM_Cubic, tangentMode=RCTM_Break, time=1, value=1}
        ]}}}
     Scale Curve=0.75
    )
    Index 0 Param 4=500
    Module.Param0WriteEnabled=true
    Module.Param1WriteEnabled=true
    Module.Param2WriteEnabled=true
    Module.Param3WriteEnabled=true
   )
   &ScaleColor(
    Scale Alpha=&MaskFloatBySpawnGroup(
     Masked Float=&FloatFromCurve(
      FloatCurve=DataInterface() {propertyValues {Curve {keys [
          {interpMode=RCIM_Cubic, leaveTangentWeight=1.182, tangentMode=RCTM_User, time=0, value=0}
          <at 0.5 1 RCIM_Cubic>
          {arriveTangent=-5.747, interpMode=RCIM_Cubic, tangentMode=RCTM_Break, time=1, value=0}
         ]}}}
      Scale Curve=&Multiply_Float(
       A=0.5
       B=&FN_LOD_Distance_Remap(Start Distance=1000, End Distance=2500, Minimum Output=0, Maximum Output=1, Clamp Output=true)
      )
     )
     Passthrough Float=&MaskFloatBySpawnGroup(
      Masked Float=&FloatFromCurve(
       FloatCurve=DataInterface() {propertyValues {Curve {keys [
           <at 0 0 RCIM_Cubic>
           {arriveTangent=1.4286, interpMode=RCIM_Cubic, leaveTangent=1.4286, time=0.2, value=1}
           {arriveTangent=-1.25, interpMode=RCIM_Cubic, leaveTangent=-1.25, time=0.7, value=1}
           <at 1 0>
          ]}}}
       Scale Curve=&Multiply_Float(
        A=0.5
        B=&FN_LOD_Distance_Remap(Start Distance=1000, End Distance=2500, Minimum Output=0, Maximum Output=1, Clamp Output=true)
       )
      )
      Passthrough Float=&MaskFloatBySpawnGroup(
       Masked Float=&FloatFromCurve(
        FloatCurve=DataInterface() {propertyValues {Curve {keys [
            {interpMode=RCIM_Cubic, tangentMode=RCTM_User, time=0, value=0}
            {interpMode=RCIM_Cubic, tangentMode=RCTM_User, time=0.5, value=1}
            {tangentMode=RCTM_User, time=1, value=0}
           ]}}}
        Scale Curve=0.25
       )
       Passthrough Float=&MaskFloatBySpawnGroup(
        Masked Float=&FloatFromCurve(
         FloatCurve=DataInterface() {propertyValues {Curve {keys [
             {arriveTangent=12.4054, interpMode=RCIM_Cubic, leaveTangent=12.4054, tangentMode=RCTM_User, time=0.125, value=0}
             <at 0.2 1 RCIM_Cubic>
             {arriveTangent=-0.7128, interpMode=RCIM_Cubic, leaveTangent=-0.7128, tangentMode=RCTM_User, time=1, value=0}
            ]}}}
         Scale Curve=&Multiply_Float(
          A=&RandomRangeFloat(Minimum=0.5, Maximum=1)
          B=&FN_LOD_Distance_Remap(Start Distance=500, End Distance=2500, Minimum Output=0, Maximum Output=1, Clamp Output=true)
         )
        )
        Passthrough Float=&MaskFloatBySpawnGroup(
         Masked Float=&FloatFromCurve(FloatCurve=DataInterface() {propertyValues {Curve {keys [
              {interpMode=RCIM_Cubic, tangentMode=RCTM_Break, time=0, value=0}
              {interpMode=RCIM_Cubic, tangentMode=RCTM_Break, time=0.16, value=0.8}
              {interpMode=RCIM_Cubic, tangentMode=RCTM_Break, time=1, value=0}
             ]}}})
         Passthrough Float=&MaskFloatBySpawnGroup(
          Masked Float=&FloatFromCurve(
           FloatCurve=DataInterface() {propertyValues {Curve {keys [{arriveTangent=1.0833, interpMode=RCIM_Cubic, time=0, value=0}, <at 0.5024 0.6194 RCIM_Cubic>, <at 1 0>]}}}
           Scale Curve=&Multiply_Float(
            A=1
            B=&FN_LOD_Distance_Remap(Start Distance=500, End Distance=2500, Minimum Output=0, Maximum Output=1, Clamp Output=true)
           )
          )
          Passthrough Float=&Multiply_Float(
           A=&FloatFromCurve(
            FloatCurve=DataInterface() {propertyValues {Curve {keys [
                <at 0 0 RCIM_Cubic>
                {interpMode=RCIM_Cubic, tangentMode=RCTM_User, time=0.1, value=0.9914}
                {interpMode=RCIM_Cubic, tangentMode=RCTM_User, time=0.7, value=1}
                <at 1 0>
               ]}}}
            Scale Curve=0.25
           )
           B=&Waveform(
            "Waveform [1]"=::NewEnumerator0
            "[1] Frequency"=&RandomRangeFloat(Minimum=5, Maximum=8)
            "[1] Amplitude Min/Max"=<0.6 1>
           )
          )
          Spawn Group Mask=5
         )
         Spawn Group Mask=4
        )
        Spawn Group Mask=3
       )
       Spawn Group Mask=2
      )
      Spawn Group Mask=1
     )
    )
    Scale RGB=&MaskVectorBySpawnGroup(
     Masked Vector=/Niagara/DynamicInputs/TypeConversions/MakeVectorFromLinearColorHSV(Color:LinearColor=&HueShiftLinearColor(
       Hue=&FloatFromCurve(FloatCurve=DataInterface() {propertyValues {Curve {keys [<at 0 0>,<at 1 0.25>]}}})
       Saturation=1
       Value=1
       Linear Color=$Particles.Initial.Color
      ))
     Passthrough Vector=&MaskVectorBySpawnGroup(
      Masked Vector=<5 5 5>
      Passthrough Vector=&MaskVectorBySpawnGroup(
       Masked Vector=<1 1 1>
       Passthrough Vector=&MaskVectorBySpawnGroup(
        Masked Vector=<1 1 1>
        Passthrough Vector=&MaskVectorBySpawnGroup(
         Masked Vector=<1 1 1>
         Passthrough Vector=&MaskVectorBySpawnGroup(
          Masked Vector=&MakeVectorFromLinearColorRGB(Color=&ColorFromCurve(ColorCurve=DataInterface() {propertyValues {
              AlphaCurve {keys [<at 0 1 RCIM_Cubic>,<at 1 0 RCIM_Cubic>]}
              BlueCurve {keys [<at 0.2176 1>, {arriveTangent=-0.1731, interpMode=RCIM_Cubic, time=0.5937, value=1}, <at 1 0.8281 RCIM_Cubic>]}
              GreenCurve {keys [<at 0.2176 1>, {arriveTangent=-0.8557, interpMode=RCIM_Cubic, time=0.5937, value=0.646}, <at 1 0.1504 RCIM_Cubic>]}
              RedCurve {keys [<at 0.2176 1>,<at 0.5937 0.1136 RCIM_Cubic>,<at 1 1 RCIM_Cubic>]}
             }}))
          Passthrough Vector=&MakeVectorFromLinearColorRGB(Color=&ColorFromCurve(ColorCurve=DataInterface() {propertyValues {
              AlphaCurve {keys [<at 0 1 RCIM_Cubic>,<at 1 0 RCIM_Cubic>]}
              BlueCurve {keys [<at 0 1 RCIM_Cubic>,<at 1 1 RCIM_Cubic>]}
              GreenCurve {keys [<at 0 0.6875 RCIM_Cubic>,<at 1 0.0414 RCIM_Cubic>]}
              RedCurve {keys [<at 0 0.1067 RCIM_Cubic>,<at 1 0.6941 RCIM_Cubic>]}
             }}))
          Spawn Group Mask=5
         )
         Spawn Group Mask=4
        )
        Spawn Group Mask=3
       )
       Spawn Group Mask=2
      )
      Spawn Group Mask=1
     )
    )
    ScaleA=true
    ScaleRGB=true
   )
   &UpdateMeshOrientation(
    Rotation Rate=&MaskFloatBySpawnGroup(
     Masked Float=0
     Passthrough Float=&MaskFloatBySpawnGroup(
      Masked Float=0.5
      Passthrough Float=&MaskFloatBySpawnGroup(
       Masked Float=0
       Passthrough Float=&MaskFloatBySpawnGroup(Masked Float=0, Passthrough Float=0, Spawn Group Mask=3)
       Spawn Group Mask=2
      )
      Spawn Group Mask=1
     )
    )
    Rotation Vector=&RandomVector(Vector Scale=1)
   ) {module=UpdateMeshOrientation001}
   &ScaleMeshSize(Scale Factor=&MaskVectorBySpawnGroup(
     Masked Vector=<1 1 1>
     Passthrough Vector=&MaskVectorBySpawnGroup(
      Masked Vector=&VectorFromFloat(Value=&FloatFromCurve(
        FloatCurve=DataInterface() {propertyValues {Curve {keys [
            {arriveTangent=9.5937, interpMode=RCIM_Cubic, leaveTangent=9.5937, tangentMode=RCTM_User, time=0, value=0}
            <at 0.2 1>
            {arriveTangentWeight=0.2134, interpMode=RCIM_Cubic, leaveTangentWeight=0.2134, tangentMode=RCTM_User, time=0.6, value=1}
            {arriveTangent=-5.5237, interpMode=RCIM_Cubic, leaveTangent=-5.5237, tangentMode=RCTM_User, time=1, value=0}
           ]}}}
        Scale Curve=4
       ))
      Passthrough Vector=&MaskVectorBySpawnGroup(
       Masked Vector=&VectorFromFloat(Value=&FloatFromCurve(
         FloatCurve=DataInterface() {propertyValues {Curve {keys [{interpMode=RCIM_Cubic, leaveTangent=1.1268, tangentMode=RCTM_Break, time=0, value=0.5}, <at 1 1>]}}}
         Scale Curve=0.5
        ))
       Passthrough Vector=&MaskVectorBySpawnGroup(
        Masked Vector=&VectorFromFloat(Value=&FloatFromCurve(
          FloatCurve=DataInterface() {propertyValues {Curve {keys [
              {arriveTangent=2.61, interpMode=RCIM_Cubic, leaveTangent=2.61, tangentMode=RCTM_User, time=0, value=0}
              {interpMode=RCIM_Cubic, leaveTangent=-0.8889, time=0.8, value=1}
             ]}}}
          Scale Curve=0.5
         ))
        Passthrough Vector=&VectorFromFloat(Value=&FloatFromCurve(FloatCurve=DataInterface() {propertyValues {Curve {keys [<at 0 1>,<at 1 1.5>]}}}))
        Spawn Group Mask=3
       )
       Spawn Group Mask=2
      )
      Spawn Group Mask=1
     )
    ))
   &SpriteRotationRate(Rotation Rate=&RandomRangeFloat(Minimum=-20, Maximum=20))
   &ScaleSpriteSize(
    Scale Sprite Size Mode=::NewEnumerator0
    Uniform Scale Factor=&MaskFloatBySpawnGroup(
     Masked Float=0
     Passthrough Float=&MaskFloatBySpawnGroup(
      Masked Float=0
      Passthrough Float=&MaskFloatBySpawnGroup(
       Masked Float=0
       Passthrough Float=&MaskFloatBySpawnGroup(
        Masked Float=0
        Passthrough Float=&MaskFloatBySpawnGroup(
         Masked Float=0
         Passthrough Float=&MaskFloatBySpawnGroup(
          Masked Float=&FloatFromCurve(
           FloatCurve=DataInterface() {propertyValues {Curve {keys [
               {interpMode=RCIM_Cubic, leaveTangent=0.9575, tangentMode=RCTM_Break, time=0, value=0.3}
               {interpMode=RCIM_Cubic, tangentMode=RCTM_Break, time=1, value=1}
              ]}}}
           Scale Curve=30
          )
          Passthrough Float=&FloatFromCurve(FloatCurve=DataInterface() {propertyValues {Curve {keys [<at 0 0 RCIM_Cubic>,<at 0.1483 1 RCIM_Cubic>,<at 1 0>]}}})
          Spawn Group Mask=5
         )
         Spawn Group Mask=4
        )
        Spawn Group Mask=3
       )
       Spawn Group Mask=2
      )
      Spawn Group Mask=1
     )
    )
   )
   &GravityForce(Gravity=&MaskVectorBySpawnGroup(
     Masked Vector=<0 0 0>
     Passthrough Vector=&MaskVectorBySpawnGroup(
      Masked Vector=<0 0 0>
      Passthrough Vector=&MaskVectorBySpawnGroup(
       Masked Vector=<0 0 0>
       Passthrough Vector=&MaskVectorBySpawnGroup(
        Masked Vector=<0 0 0>
        Passthrough Vector=&MaskVectorBySpawnGroup(
         Masked Vector=<0 0 0>
         Passthrough Vector=&MaskVectorBySpawnGroup(Masked Vector=<0 0 0>, Passthrough Vector=<0 0 50>, Spawn Group Mask=5)
         Spawn Group Mask=4
        )
        Spawn Group Mask=3
       )
       Spawn Group Mask=2
      )
      Spawn Group Mask=1
     )
    ))
   &SolveForcesAndVelocity(Clamp Velocity=true, Limit Acceleration=false, Speed Limit=100)
  ]
  rendererClasses [&NiagaraMeshRendererProperties, &NiagaraSpriteRendererProperties]
  renderers [
   {rendererClass=&NiagaraMeshRendererProperties, rendererIndex=0}
   {rendererClass=&NiagaraMeshRendererProperties, rendererIndex=1}
   {rendererClass=&NiagaraMeshRendererProperties, rendererIndex=2}
   {rendererClass=&NiagaraMeshRendererProperties, rendererIndex=3}
   {rendererClass=&NiagaraMeshRendererProperties, rendererIndex=4}
   {rendererClass=&NiagaraSpriteRendererProperties, rendererIndex=5}
   {rendererClass=&NiagaraSpriteRendererProperties, rendererIndex=6}
   {rendererClass=&NiagaraSpriteRendererProperties, rendererIndex=7}
  ]
 }
 Light {
  emitterUpdateScript [
   &EmitterState(
    Enable Distance Culling=true
    Inactive Response=::NewEnumerator1
    Life Cycle Mode=::NewEnumerator1
    Loop Behavior=::NewEnumerator1
    Loop Duration Mode=::NewEnumerator1
    Max Distance Response=::SleepAndClearParticles
    MaxDistance=15000
    Module.bUseMaxDistance=true
    Module.bUseMinDistance=false
    Reset Age On Awaken=true
    Scalability Mode=::NewEnumerator1
    UseLoopDelay=false
   )
   &SpawnBurst_Instantaneous(Use Spawn Probability=false, UseLoopCountLimit=false)
  ]
  particleSpawnScript [&InitializeParticle(
    Color=<rgba 100 5.098 78.0392 1>
    Lifetime=2
    Sprite Size Mode=::NewEnumerator3
    Uniform Sprite Size=500
    UsePositionOffset=false
    Write Material Random=true
   )]
  particleUpdateScript [
   &ParticleState(Kill Particles When Lifetime Has Elapsed=false)
   &ScaleColor(
    Scale Alpha=&FloatFromCurve(
     FloatCurve=DataInterface() {propertyValues {Curve {defaultValue=1, keys [<at 0 0 RCIM_Cubic>, {tangentMode=RCTM_User, time=1, value=1}]}}}
     Scale Curve=&FloatFromCurve(FloatCurve=DataInterface() {propertyValues {Curve {keys [<at 12000 1 RCIM_Cubic>,<at 15000 0>]}}}, CurveIndex=$Engine.Owner.LODDistance)
    )
    ScaleA=true
    ScaleRGB=false
   )
  ]
  rendererClasses [&NiagaraSpriteRendererProperties, &NiagaraLightRendererProperties]
  renderers [
   {rendererClass=&NiagaraSpriteRendererProperties}
   {rendererClass=&NiagaraLightRendererProperties, rendererIndex=1}
  ]
 }
 Mesh_Swirls {
  bEnabled=false
  emitterUpdateScript [
   &EmitterState()
   &SpawnRate(SpawnRate=10, Use Spawn Probability=false)
  ]
  particleSpawnScript [
   &InitializeParticle(
    Color=&SelectLinearColorFromArray(
     Array Sampling Mode=::NewEnumerator1
     Direct Array Index=&RandomRangeInt(Minimum=0, Maximum=1)
     Color Selection Array=DataInterface() {propertyValues {ColorData [<rgba 0.3199 0.5916 1 1>, <rgba 0.9443 0.3385 1 1>]}}
     Random Seed=?
    )
    Lifetime=&RandomRangeFloat(Minimum=3, Maximum=4)
    Mesh Scale=&RandomRangeVector(Minimum=<2 1.5 1.5>, Maximum=<3 8 8>)
    Mesh Scale Mode=::NewEnumerator4
    UsePositionOffset=false
    Write Material Random=true
   ) {module=InitializeParticle001}
   &InitialMeshOrientation(
    Rotation=&RandomRangeVector(Minimum=<-1 -1 -1>, Maximum=<1 1 1>)
    Rotation Coordinate Space=::NewEnumerator2
    Use Orientation Vector=true
    Use Rotation Vector=true
   ) {module=InitialMeshOrientation001}
   &ApplyInitialForces()
   &AddVelocity(Module.Use Falloff Distance=false, Override Seed=false, Velocity Mode=::NewEnumerator1, Velocity Speed=5)
  ]
  particleUpdateScript [
   &ParticleState()
   &DynamicMaterialParameters(
    Index 0 Param 1=&RandomRangeFloat(Minimum=0, Maximum=1)
    Index 0 Param 2=&FloatFromCurve(FloatCurve=DataInterface() {propertyValues {Curve {
        defaultValue=1
        keys [
         {interpMode=RCIM_Cubic, tangentMode=RCTM_User, time=0, value=1}
         <at 0.5 0 RCIM_Cubic>
         {interpMode=RCIM_Cubic, tangentMode=RCTM_User, time=1, value=1}
        ]
       }}})
    Index 0 Param 3=0.5
    Index 0 Param 4=&FloatFromCurve(
     FloatCurve=DataInterface() {propertyValues {Curve {keys [<at 0 0.3 RCIM_Cubic>, {arriveTangent=-3.2559, interpMode=RCIM_Cubic, leaveTangent=-3.2559, tangentMode=RCTM_User, time=1, value=-1}]}}}
     Scale Curve=0.5
    )
    Module.Param0WriteEnabled=true
    Module.Param1WriteEnabled=true
    Module.Param2WriteEnabled=true
    Module.Param3WriteEnabled=true
   )
   &ScaleColor(
    Scale Alpha=&FloatFromCurve(
     FloatCurve=DataInterface() {propertyValues {Curve {keys [
         {arriveTangent=12.4054, interpMode=RCIM_Cubic, leaveTangent=12.4054, tangentMode=RCTM_User, time=0.125, value=0}
         <at 0.2 1 RCIM_Cubic>
         {arriveTangent=-0.7128, interpMode=RCIM_Cubic, leaveTangent=-0.7128, tangentMode=RCTM_User, time=1, value=0}
        ]}}}
     Scale Curve=&Multiply_Float(
      A=&RandomRangeFloat(Minimum=0.5, Maximum=1)
      B=&FN_LOD_Distance_Remap(Start Distance=500, End Distance=2500, Minimum Output=0, Maximum Output=1, Clamp Output=true)
     )
    )
    ScaleA=true
    ScaleRGB=true
   )
   &ScaleMeshSize(Scale Factor=&VectorFromFloat(Value=&FloatFromCurve(
      FloatCurve=DataInterface() {propertyValues {Curve {keys [
          {arriveTangent=2.61, interpMode=RCIM_Cubic, leaveTangent=2.61, tangentMode=RCTM_User, time=0, value=0}
          {interpMode=RCIM_Cubic, leaveTangent=-0.8889, time=0.8, value=1}
         ]}}}
      Scale Curve=0.5
     )))
   &SolveForcesAndVelocity(Clamp Velocity=false, Limit Acceleration=false)
  ]
  rendererClasses [&NiagaraMeshRendererProperties]
  renderers [{rendererClass=&NiagaraMeshRendererProperties, rendererIndex=0}]
 }
 Motes {
  bEnabled=false
  simTarget=GPUComputeSim
  emitterUpdateScript [
   &EmitterState()
   &SpawnRate(SpawnRate=80, Use Spawn Probability=false)
  ]
  particleSpawnScript [
   &InitializeParticle(
    Lifetime Mode=::NewEnumerator1
    Position Mode=::NewEnumerator1
    Sprite Rotation Mode=::NewEnumerator2
    Sprite Size Mode=::NewEnumerator1
    Uniform Sprite Size Max=100
    Uniform Sprite Size Min=50
    UsePositionOffset=false
    Write Material Random=true
   )
   /SetVariables(Particles.Position=&MakeVector(
     X=&FloatFromCurve(
      FloatCurve=DataInterface() {propertyValues {Curve {keys [
          {arriveTangent=4.8627, interpMode=RCIM_Cubic, leaveTangent=0.4743, tangentMode=RCTM_User, time=0, value=0.07}
          {arriveTangent=2.7374, interpMode=RCIM_Cubic, leaveTangent=2.7374, tangentMode=RCTM_User, time=0.2781, value=0.3179}
          {arriveTangent=0.7984, interpMode=RCIM_Cubic, leaveTangent=0.7984, tangentMode=RCTM_User, time=0.6554, value=1.0037}
          {arriveTangent=1.8113, interpMode=RCIM_Cubic, leaveTangent=0.022, tangentMode=RCTM_User, time=1, value=0.75}
         ]}}}
      CurveIndex=$Particles.RandomPosition
      Scale Curve=500
     )
     Y=&Add_Float(
      A=&FloatFromCurve(
       FloatCurve=DataInterface() {propertyValues {Curve {keys [
           {arriveTangent=-2.6903, interpMode=RCIM_Cubic, leaveTangent=-2.6903, tangentMode=RCTM_User, time=-0.0027, value=0}
           {interpMode=RCIM_Cubic, tangentMode=RCTM_User, time=0.3, value=-1}
           {interpMode=RCIM_Cubic, leaveTangent=1.4239, time=0.7, value=1}
           <at 1 0 RCIM_Cubic>
          ]}}}
       CurveIndex=$Particles.RandomPosition
       Scale Curve=0
      )
      B=&FloatFromCurve(
       FloatCurve=DataInterface() {propertyValues {Curve {keys [
           {arriveTangent=-0.8464, interpMode=RCIM_Cubic, time=0, value=0.25}
           {arriveTangent=1.1407, interpMode=RCIM_Cubic, leaveTangent=1.1407, tangentMode=RCTM_User, time=1, value=1}
          ]}}}
       CurveIndex=$Particles.RandomPosition
      )
     )
     Z=&FloatFromCurve(
      FloatCurve=DataInterface() {propertyValues {Curve {keys [
          {arriveTangent=7.2867, interpMode=RCIM_Cubic, time=0, value=1}
          {arriveTangent=-4.3515, interpMode=RCIM_Cubic, leaveTangent=-4.3515, tangentMode=RCTM_User, time=0.2171, value=0.5268}
          {arriveTangent=-0.0104, interpMode=RCIM_Cubic, leaveTangent=-0.0104, tangentMode=RCTM_User, time=0.45, value=0}
          {arriveTangent=4.8334, interpMode=RCIM_Cubic, leaveTangent=4.8334, tangentMode=RCTM_User, time=0.8116, value=0.579}
          {arriveTangent=-2.1665, interpMode=RCIM_Cubic, leaveTangent=-2.1665, tangentMode=RCTM_User, time=0.9827, value=0.7}
         ]}}}
      CurveIndex=$Particles.RandomPosition
     )
    )) {enabled=false, isSetParameters=true}
   /SetVariables(Particles.Sparkle:NiagaraFloat=&UniformAOrBFloat(A=0, Distribution Weight=0.2, Evaluation Type=::NewEnumerator0, Recalculate Random Each Loop=true, Override Seed=false)) {isSetParameters=true, module=SetVariables_#2}
   &ShapeLocation(MaskSpawnGroup=false, Override Seed=false, Sphere Radius=500, UseOwnerScale=true)
   &AddVelocity(Module.Use Falloff Distance=false, Override Seed=false, Velocity Mode=::NewEnumerator1, Velocity Speed=250)
  ]
  particleUpdateScript [
   &ParticleState()
   &DynamicMaterialParameters(
    Index 0 Param 1=2
    Index 0 Param 2=1
    Index 0 Param 3=0.5
    Module.Param0WriteEnabled=true
    Module.Param1WriteEnabled=true
    Module.Param2WriteEnabled=true
    Module.Param3WriteEnabled=true
   )
   &ScaleColor(
    Scale Alpha=&Multiply_Float(
     A=&FloatFromCurve(
      FloatCurve=DataInterface() {propertyValues {Curve {keys [
          <at 0 0 RCIM_Cubic>
          {interpMode=RCIM_Cubic, tangentMode=RCTM_User, time=0.1, value=0.9914}
          {interpMode=RCIM_Cubic, tangentMode=RCTM_User, time=0.7, value=1}
          <at 1 0>
         ]}}}
      Scale Curve=0.25
     )
     B=&Waveform(
      "Waveform [1]"=::NewEnumerator0
      "[1] Frequency"=&RandomRangeFloat(Minimum=5, Maximum=8)
      "[1] Amplitude Min/Max"=<0.6 1>
     )
    )
    Scale RGB=&MakeVectorFromLinearColorRGB(Color=&ColorFromCurve(ColorCurve=DataInterface() {propertyValues {
        AlphaCurve {keys [<at 0 1 RCIM_Cubic>,<at 1 0 RCIM_Cubic>]}
        BlueCurve {keys [<at 0 1 RCIM_Cubic>,<at 1 1 RCIM_Cubic>]}
        GreenCurve {keys [<at 0 0.6875 RCIM_Cubic>,<at 1 0.0414 RCIM_Cubic>]}
        RedCurve {keys [<at 0 0.1067 RCIM_Cubic>,<at 1 0.6941 RCIM_Cubic>]}
       }}))
    ScaleA=true
    ScaleRGB=true
   )
   &ScaleSpriteSize(Scale Factor=&Vector2DFromFloat(Value=&FloatFromCurve(FloatCurve=DataInterface() {propertyValues {Curve {keys [<at 0 0 RCIM_Cubic>,<at 0.1483 1 RCIM_Cubic>,<at 1 0>]}}})))
   &Drag(Drag=0, Use Linear Drag=true, Use Rotational Drag=true)
   &GravityForce(Gravity=<0 0 50>)
   &CurlNoiseForce(Bias Noise=false, Noise Frequency=5, Noise Strength=2500, Pan Noise=false)
   &SolveForcesAndVelocity(Clamp Velocity=true, Limit Acceleration=false, Speed Limit=50)
  ]
  rendererClasses [&NiagaraSpriteRendererProperties]
  renderers [
   {rendererClass=&NiagaraSpriteRendererProperties}
   {rendererClass=&NiagaraSpriteRendererProperties, rendererIndex=1}
  ]
 }
 PlasmaWave {bEnabled=false, rendererClasses [], renderers []}
 Purple_Mesh_Glow_Burst {
  bEnabled=false
  emitterUpdateScript [
   &EmitterState(
    Enable Distance Culling=true
    MaxDistance=30000
    Module.bUseMaxDistance=true
    Module.bUseMinDistance=false
    Scalability Mode=::NewEnumerator1
   )
   &SpawnRate(SpawnRate=2, Use Spawn Probability=false)
  ]
  particleSpawnScript [
   /Niagara/Modules/Spawn/Initialization/InitializeParticle(
    Color=<rgba 1.2844 1.0153 4 0.75>
    Lifetime=&RandomRangeFloat(Minimum=2, Maximum=3)
    Mesh Scale=&VectorFromFloat(Value=&self:StageSelection_02(
      Stage:NiagaraInt32=$User.Stage
      Stage 0:NiagaraFloat=0
      Stage 1:NiagaraFloat=0.9
      Stage 2:NiagaraFloat=1.2
      Stage 3:NiagaraFloat=1.5
      Stage 4:NiagaraFloat=2.2
      Stage 5:NiagaraFloat=2.5
      Stage 6:NiagaraFloat=4
     ))
    Module.Write Color=true
    Module.Write Lifetime=true
    Module.Write Mass=true
    Module.Write Material Random=true
    Module.Write Position=true
    Module.Write Scale=true
    Module.Write SpriteRotation=false
    Module.Write SpriteSize=false
    Sprite Rotation=&RandomRangeFloat(Minimum=0, Maximum=360)
    Sprite Size=&Vector2DFromFloat(Value=&RandomRangeFloat(Minimum=3000, Maximum=3200))
   )
   &InitialMeshOrientation(Use Rotation Vector=false)
  ]
  particleUpdateScript [
   &ParticleState()
   &DynamicMaterialParameters(
    Index 0 Param 1=1.001
    Index 0 Param 2=1
    Index 0 Param 3=1
    Index 0 Param 4=1
    Module.Param0WriteEnabled=true
    Module.Param1WriteEnabled=true
    Module.Param2WriteEnabled=true
    Module.Param3WriteEnabled=true
   )
   &ScaleColor(
    Scale Alpha=&FloatFromCurve(
     FloatCurve=DataInterface() {propertyValues {Curve {keys [
         {interpMode=RCIM_Cubic, tangentMode=RCTM_User, time=0, value=0}
         {interpMode=RCIM_Cubic, tangentMode=RCTM_User, time=0.5, value=1}
         {tangentMode=RCTM_User, time=1, value=0}
        ]}}}
     Scale Curve=0.25
    )
    ScaleA=true
    ScaleRGB=true
   )
   &ScaleMeshSize(Scale Factor=&VectorFromFloat(Value=&FloatFromCurve(
      FloatCurve=DataInterface() {propertyValues {Curve {keys [{interpMode=RCIM_Cubic, leaveTangent=1.1268, tangentMode=RCTM_Break, time=0, value=0.5}, <at 1 1>]}}}
      Scale Curve=0.5
     )))
  ]
  rendererClasses [&NiagaraMeshRendererProperties]
  renderers [{rendererClass=&NiagaraMeshRendererProperties, rendererIndex=0}]
 }
 Shimmers_Top {
  bEnabled=false
  simTarget=GPUComputeSim
  emitterUpdateScript [
   &EmitterState()
   &SpawnRate(SpawnRate=2, Use Spawn Probability=false)
  ]
  particleSpawnScript [
   /Niagara/Modules/Spawn/Initialization/V2/InitializeParticle(
    Lifetime Min=2
    Lifetime Mode=::NewEnumerator1
    Position Mode=::NewEnumerator1
    Sprite Rotation Mode=::NewEnumerator2
    Sprite Size=&RandomRangeVector2D(Minimum=<85 85>, Maximum=<125 125>, Evaluation Type=::NewEnumerator1)
    Sprite Size Mode=::NewEnumerator4
    UsePositionOffset=false
    Write Material Random=true
   )
   &AddVelocity(Velocity=&RandomRangeVector(Minimum=<-100 -100 -100>, Maximum=<100 100 100>))
   &DynamicMaterialParameters(
    Index 0 Param 1=&RandomRangeFloat(Minimum=0, Maximum=1, Evaluation Type=::NewEnumerator1)
    Index 0 Param 2=&RandomRangeFloat(Minimum=0.4, Maximum=0.6, Evaluation Type=::NewEnumerator1)
    Index 0 Param 3=0.3
    Module.Index0Float0Write=true
    Module.Index0Float1Write=true
    Module.Index0Float2Write=true
    Module.Index0Float3Write=true
   )
  ]
  particleUpdateScript [
   &ParticleState()
   &SolveForcesAndVelocity(Clamp Velocity=false, Limit Acceleration=false)
   &SpriteRotationRate(Rotation Rate=&RandomRangeFloat(Minimum=-20, Maximum=20))
   &ScaleColor(
    Scale Alpha=&FloatFromCurve(
     FloatCurve=DataInterface() {propertyValues {Curve {keys [{arriveTangent=1.0833, interpMode=RCIM_Cubic, time=0, value=0}, <at 0.5024 0.6194 RCIM_Cubic>, <at 1 0>]}}}
     Scale Curve=&Multiply_Float(
      A=1
      B=&FN_LOD_Distance_Remap(Start Distance=500, End Distance=2500, Minimum Output=0, Maximum Output=1, Clamp Output=true)
     )
    )
    Scale RGB=&MakeVectorFromLinearColorRGB(Color=&ColorFromCurve(ColorCurve=DataInterface() {propertyValues {
        AlphaCurve {keys [<at 0 1 RCIM_Cubic>,<at 1 0 RCIM_Cubic>]}
        BlueCurve {keys [<at 0.2176 1>, {arriveTangent=-0.1731, interpMode=RCIM_Cubic, time=0.5937, value=1}, <at 1 0.8281 RCIM_Cubic>]}
        GreenCurve {keys [<at 0.2176 1>, {arriveTangent=-0.8557, interpMode=RCIM_Cubic, time=0.5937, value=0.646}, <at 1 0.1504 RCIM_Cubic>]}
        RedCurve {keys [<at 0.2176 1>,<at 0.5937 0.1136 RCIM_Cubic>,<at 1 1 RCIM_Cubic>]}
       }}))
    ScaleA=true
    ScaleRGB=true
   )
   &ScaleSpriteSize(
    Uniform Curve Scale=30
    Uniform Curve Sprite Scale=DataInterface() {propertyValues {Curve {keys [
        {arriveTangent=0.8917, interpMode=RCIM_Cubic, leaveTangent=0.8917, tangentMode=RCTM_User, time=0, value=0.3}
        {interpMode=RCIM_Cubic, tangentMode=RCTM_Break, time=1, value=1}
       ]}}}
   )
  ]
  rendererClasses [&NiagaraSpriteRendererProperties]
  renderers [{rendererClass=&NiagaraSpriteRendererProperties}]
 }
 TinyFrags {
  bEnabled=false
  simTarget=GPUComputeSim
  emitterUpdateScript [
   &EmitterState(Enable Distance Culling=true, Module.bUseMaxDistance=true, Module.bUseMinDistance=false, Scalability Mode=::NewEnumerator1)
   &SpawnRate(SpawnRate=125, Use Spawn Probability=false)
  ]
  particleSpawnScript [
   /Niagara/Modules/Spawn/Initialization/InitializeParticle(
    Color=&RandomRangeLinearColor(Minimum=<rgba 1 0.0336 0.9623 1>, Maximum=<rgba 0.7292 0.745 1 1>)
    Lifetime=&RandomRangeFloat(Minimum=0.5, Maximum=1, Evaluation Type=::NewEnumerator1)
    Mesh Scale=&RandomRangeVector(Minimum=<0.4 0.4 1>, Maximum=<0.7 0.7 1>, Evaluation Type=::NewEnumerator1)
    Module.Write Color=true
    Module.Write Lifetime=true
    Module.Write Mass=true
    Module.Write Material Random=true
    Module.Write Position=false
    Module.Write Scale=true
    Module.Write SpriteRotation=false
    Module.Write SpriteSize=false
   )
   &ShapeLocation(MaskSpawnGroup=false, Offset Mode=::NewEnumerator0, Override Seed=false, Sphere Radius=500, UseOwnerScale=true)
   &AddVelocity(
    Module.Use Falloff Distance=false
    Override Seed=false
    Velocity Mode=::NewEnumerator1
    Velocity Speed=&RandomRangeFloat(Minimum=250, Maximum=500)
   ) {module=AddVelocity001}
   &InitialMeshOrientation(Use Rotation Vector=false)
  ]
  particleUpdateScript [
   &ParticleState()
   &CurlNoiseForce(Noise Frequency=2, Noise Strength=250, Pan Noise=false)
   &Drag(Use Linear Drag=true, Use Rotational Drag=false)
   &CurlNoiseForce(Bias Noise=false, Noise Frequency=5, Noise Strength=2500, Pan Noise=false) {module=CurlNoiseForce001}
   &SolveForcesAndVelocity(Clamp Velocity=true, Limit Acceleration=false, Speed Limit=100)
   &ScaleColor(
    Scale Alpha=&FloatFromCurve(
     FloatCurve=DataInterface() {propertyValues {Curve {keys [
         <at 0 0 RCIM_Cubic>
         {arriveTangent=1.4286, interpMode=RCIM_Cubic, leaveTangent=1.4286, time=0.2, value=1}
         {arriveTangent=-1.25, interpMode=RCIM_Cubic, leaveTangent=-1.25, time=0.7, value=1}
         <at 1 0>
        ]}}}
     Scale Curve=&Multiply_Float(
      A=0.5
      B=&FN_LOD_Distance_Remap(Start Distance=1000, End Distance=2500, Minimum Output=0, Maximum Output=1, Clamp Output=true)
     )
    )
    Scale RGB=<5 5 5>
    ScaleA=true
    ScaleRGB=true
   )
   &UpdateMeshOrientation(Rotation Rate=0.5, Rotation Vector=&RandomVector(Vector Scale=1)) {module=UpdateMeshOrientation001}
   &ScaleMeshSize(Scale Factor=&VectorFromFloat(Value=&FloatFromCurve(
      FloatCurve=DataInterface() {propertyValues {Curve {keys [
          {arriveTangent=9.5937, interpMode=RCIM_Cubic, leaveTangent=9.5937, tangentMode=RCTM_User, time=0, value=0}
          <at 0.2 1>
          {arriveTangentWeight=0.2134, interpMode=RCIM_Cubic, leaveTangentWeight=0.2134, tangentMode=RCTM_User, time=0.6, value=1}
          {arriveTangent=-5.5237, interpMode=RCIM_Cubic, leaveTangent=-5.5237, tangentMode=RCTM_User, time=1, value=0}
         ]}}}
      Scale Curve=4
     )))
  ]
  rendererClasses [&NiagaraMeshRendererProperties]
  renderers [{rendererClass=&NiagaraMeshRendererProperties, rendererIndex=0}]
 }
}
summary {
 systemName=NS_TechHQ_Shard
 emitters [
  {emitterName=Energy, rendererClasses [&NiagaraMeshRendererProperties, &NiagaraSpriteRendererProperties]}
  {bEnabled=false, emitterName=TinyFrags, simTarget=GPUComputeSim, rendererClasses [&NiagaraMeshRendererProperties]}
  {bEnabled=false, emitterName=Purple_Mesh_Glow_Burst, rendererClasses [&NiagaraMeshRendererProperties]}
  {bEnabled=false, emitterName=Shimmers_Top, simTarget=GPUComputeSim, rendererClasses [&NiagaraSpriteRendererProperties]}
  {bEnabled=false, emitterName=Mesh_Swirls, rendererClasses [&NiagaraMeshRendererProperties]}
  {bEnabled=false, emitterName=PlasmaWave, rendererClasses []}
  {bEnabled=false, emitterName=Motes, simTarget=GPUComputeSim, rendererClasses [&NiagaraSpriteRendererProperties]}
  {emitterName=Light, rendererClasses [&NiagaraSpriteRendererProperties, &NiagaraLightRendererProperties]}
 ]
 userVariables [
  User.Fade out {type=NiagaraFloat, defaultValue=0.3, description=""}
  User.Stage {type=NiagaraInt32, defaultValue=6}
 ]
}
